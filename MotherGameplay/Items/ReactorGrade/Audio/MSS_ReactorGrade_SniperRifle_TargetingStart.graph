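# One-shot mono MetaSound for the ReactorGrade sniper rifle's targeting start sound, using targeting attenuation and the weapon foley sound class.
# On play, a random pick from a wave array variable plays through MSP_RandomModulator, which varies its pitch and gain.
# A fixed targeting-start UI wave plays alongside it at boosted gain, mixed into the mono output.
# A trigger accumulator signals finished once both wave players finish.
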MetaSoundSource /MotherGameplay/Items/ReactorGrade/Audio/MSS_ReactorGrade_SniperRifle_TargetingStart
inputs [UE.Source.OnPlay {type=Trigger}]
interfaces [UE.OutputFormat.Mono:1.0, UE.Source.OneShot:1.0, UE.Source:1.1]
outputs [
 "UE.OutputFormat.Mono.Audio:0" {type=Audio}
 UE.Source.OneShot.OnFinished {type=Trigger}
]
pages [Default {
  input_defaults {UE.Source.OnPlay=null}
  nodes [
   ArrayRandomGet_0 = Array.Random Get:WaveAsset:Array(Next=^UE.Source.OnPlay, In Array=VariableAccessor.WaveAsset:Array(Variable=InitVariable.WaveAsset:Array().Variable).Value)
   n1 = /FortMetasound/SoundCue/MSP_RandomModulator(In=@ArrayRandomGet_0.On Next)
   WavePlayer_0 = UE.Wave Player:Mono(Wave Asset=@ArrayRandomGet_0.Value, Play=@n1.Out, Pitch Shift=@n1.Pitch Out)
   WavePlayer_1 = UE.Wave Player:Mono(Play=^UE.Source.OnPlay) {also {defaults {Wave Asset=/MotherGameplay/Items/ReactorGrade/Audio/SoundWaves/ReactorGrade_Targeting_Start_UI_01}}}
   AudioMixerAudioMixerMono2_0 = "AudioMixer.Audio Mixer (Mono, 2)"(Gain 0=@n1.Volume Out, In 0=@WavePlayer_0.Out Mono, In 1=@WavePlayer_1.Out Mono) {also {defaults {Gain 1=1.8}}}
   TriggerAccumulatorTriggerAccumulate2_0 = "TriggerAccumulator.Trigger Accumulate (2)"(In 0=@WavePlayer_0.On Finished, In 1=@WavePlayer_1.On Finished)
  ]
  links [
   {dst="@UE.OutputFormat.Mono.Audio:0", src=AudioMixerAudioMixerMono2_0.Out}
   {dst="@UE.Source.OneShot.OnFinished", src=TriggerAccumulatorTriggerAccumulate2_0.Out}
  ]
 }]
properties {
 assetImportData=&self:AssetImportData
 attenuationSettings=&/ATT_ReactorGrade_Targeting
 maxDistance=2700
 sampleRate=48000
 soundClassObject=&Fort_WeaponFoley
 volume=0.05
}
