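# Sound cue for the drop sound of the ReactorGrade item.
# A Random node picks one of four equally weighted drop wave variations and feeds it through a Modulator at the root.
# It uses shared weapon-drop attenuation and concurrency assets, is assigned to the loot collect sound class, and has a slightly boosted volume.
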SoundCue /MotherGameplay/Items/ReactorGrade/Audio/SC_ReactorGrade_Drop
root=Modulator_0
nodes [
 {id=Modulator_0, children [Random_1]}
 {id=Random_1, children [WavePlayer_1,WavePlayer_5,WavePlayer_6,WavePlayer_7], properties {weights [1,1,1,1]}}
 {id=WavePlayer_1, properties {soundWaveAssetPtr=/MotherGameplay/Items/ReactorGrade/Audio/SoundWaves/ReactorGrade_Drop_04}}
 {id=WavePlayer_5, properties {soundWaveAssetPtr=/MotherGameplay/Items/ReactorGrade/Audio/SoundWaves/ReactorGrade_Drop_03}}
 {id=WavePlayer_6, properties {soundWaveAssetPtr=/MotherGameplay/Items/ReactorGrade/Audio/SoundWaves/ReactorGrade_Drop_02}}
 {id=WavePlayer_7, properties {soundWaveAssetPtr=/MotherGameplay/Items/ReactorGrade/Audio/SoundWaves/ReactorGrade_Drop_01}}
]
properties {
 attenuationSettings=&Fort_Weapon_Drop_Att
 maxDistance=3000
 soundClassObject=&Fort_Loot_Collect
 volumeMultiplier=1.2
 concurrencyOverrides {resolutionRule=PreventNew}
 concurrencySet [&Weapon_Drop_StopOldest]
}
