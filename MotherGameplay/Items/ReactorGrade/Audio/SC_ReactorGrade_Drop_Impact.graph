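# Sound cue for the impact sound of a dropped ReactorGrade item, routed to the loot-collect sound class.
# A Random node picks between two impact waves with equal weight, then feeds a Modulator left at its default settings.
# The cue overrides attenuation for a short falloff range and caps concurrency at a few instances, preventing new plays beyond that limit.
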SoundCue /MotherGameplay/Items/ReactorGrade/Audio/SC_ReactorGrade_Drop_Impact
root=Modulator_0
nodes [
 {id=Modulator_0, children [Random_0]}
 {id=Random_0, children [WavePlayer_0,WavePlayer_1], properties {weights [1,1]}}
 {id=WavePlayer_0, properties {soundWaveAssetPtr=/MotherGameplay/Items/ReactorGrade/Audio/SoundWaves/ReactorGrade_Drop_Impact_01}}
 {id=WavePlayer_1, properties {soundWaveAssetPtr=/MotherGameplay/Items/ReactorGrade/Audio/SoundWaves/ReactorGrade_Drop_Impact_02}}
]
properties {
 bOverrideAttenuation=true
 bOverrideConcurrency=true
 maxDistance=1000
 soundClassObject=&Fort_Loot_Collect
 volumeMultiplier=1
 attenuationOverrides {falloffDistance=1000, attenuationShapeExtents {x=0}}
 concurrencyOverrides {maxCount=3, resolutionRule=PreventNew}
}
