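# ReactorGrade first-person weapon fire sound cue for Athena, using first-person gunfire attenuation and concurrency.
# A short delay and a pitch modulator feed a quality-level switch.
# The first two quality slots mix a random pick of three close fire waves with a quieter, one-second-delayed targeting beam.
# The last two slots play only a single fixed close fire wave, dropping the variation and beam layer.
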SoundCue /MotherGameplay/Items/ReactorGrade/Audio/SC_ReactorGrade_Fire_Cue_1P_Athena
root=Delay_0
nodes [
 {id=Delay_0, children [Modulator_0], properties {delayMax=0.07, delayMin=0.07}}
 {id=Delay_1, children [WavePlayer_1], properties {delayMax=1, delayMin=1}}
 {id=Mixer_0, children [Random_2,Delay_1], properties {inputVolume [1,0.3]}}
 {id=Modulator_0, children [QualityLevel_0], properties {pitchMin=1, volumeMax=1, volumeMin=1}}
 {id=QualityLevel_0, children [Mixer_0, Mixer_0, WavePlayer_10, WavePlayer_10]}
 {id=Random_2, children [WavePlayer_4,WavePlayer_8,WavePlayer_9], properties {weights [1,1,1]}}
 {id=WavePlayer_1, properties {soundWaveAssetPtr=/MotherGameplay/Items/ReactorGrade/Audio/SoundWaves/ReactorGrade_TargetingBeam_1P_02}}
 {id=WavePlayer_10, properties {soundWaveAssetPtr=/MotherGameplay/Items/ReactorGrade/Audio/SoundWaves/ReactorGrade_Fire_Close_02}}
 {id=WavePlayer_4, properties {soundWaveAssetPtr=/MotherGameplay/Items/ReactorGrade/Audio/SoundWaves/ReactorGrade_Fire_Close_01}}
 {id=WavePlayer_8, properties {soundWaveAssetPtr=&ReactorGrade_Fire_Close_02}}
 {id=WavePlayer_9, properties {soundWaveAssetPtr=/MotherGameplay/Items/ReactorGrade/Audio/SoundWaves/ReactorGrade_Fire_Close_03}}
]
properties {
 attenuationSettings=&Fort_Weapon_Projectile_FirstPerson_Athena_Att
 maxDistance=2097152
 soundClassObject=&Fort_WeaponRanged_Fire_1P_PSM
 volumeMultiplier=1
 attenuationOverrides {distanceAlgorithm=NaturalSound, falloffDistance=4000, attenuationShapeExtents {x=1000}}
 concurrencyOverrides {resolutionRule=PreventNew}
 concurrencySet [&Weapon_Gunfire_FirstPerson_Conc]
}
