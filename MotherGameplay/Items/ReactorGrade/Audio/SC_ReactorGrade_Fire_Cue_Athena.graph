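# Fire sound cue for the ReactorGrade weapon, mixing a third-person close layer with a farther layer under gunfire concurrency limits.
# The farther layer crossfades medium and distant sounds by distance, and each of the three bands has its own attenuation.
# Quality switches play one of three random variations in their first two slots and a single fixed wave in the other two.
# A short fixed delay precedes the whole mix.
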SoundCue /MotherGameplay/Items/ReactorGrade/Audio/SC_ReactorGrade_Fire_Cue_Athena
root=Delay_0
nodes [
 {id=Attenuation_80, children [Modulator_0], properties {attenuationSettings=&Fort_Weapon_Projectile_Close_Athena_Att}}
 {
  id=Attenuation_81
  children [Modulator_13]
  properties {attenuationSettings=&Fort_Weapon_Projectile_Medium_Mono_Athena_Att}
 }
 {
  id=Attenuation_82
  children [Modulator_1]
  properties {attenuationSettings=&Fort_Weapon_Projectile_Distant_Mono_Athena_Att}
 }
 {id=Delay_0, children [Mixer_15], properties {delayMax=0.05, delayMin=0.05}}
 {
  id=DistanceCrossFade_25
  children [Attenuation_81, Attenuation_82]
  properties {crossFadeInput [
    {fadeInDistanceEnd=4000, fadeInDistanceStart=3000, fadeOutDistanceEnd=7000, fadeOutDistanceStart=5000, volume=1}
    {fadeInDistanceEnd=8000, fadeInDistanceStart=5000, fadeOutDistanceEnd=25000, fadeOutDistanceStart=20000, volume=1}
   ]}
 }
 {id=Mixer_0, children [Random_0], properties {inputVolume [4]}}
 {id=Mixer_1, children [Mixer_2], properties {inputVolume [2.25]}}
 {id=Mixer_15, children [SoundClass_55, SoundClass_56], properties {inputVolume [1.3,1]}}
 {id=Mixer_2, children [Random_3], properties {inputVolume [2]}}
 {id=Mixer_3, children [Mixer_5], properties {inputVolume [2.25]}}
 {id=Mixer_4, children [WavePlayer_3], properties {inputVolume [4]}}
 {id=Mixer_5, children [WavePlayer_15], properties {inputVolume [2]}}
 {id=Modulator_0, children [QualityLevel_0], properties {pitchMin=1, volumeMax=1, volumeMin=1}}
 {id=Modulator_1, children [QualityLevel_3], properties {volumeMax=0.7, volumeMin=0.7}}
 {id=Modulator_13, children [QualityLevel_4]}
 {id=QualityLevel_0, children [Random_2, Random_2, WavePlayer_10, WavePlayer_10]}
 {id=QualityLevel_3, children [Mixer_1,Mixer_1,Mixer_3,Mixer_3]}
 {id=QualityLevel_4, children [Mixer_0,Mixer_0,Mixer_4,Mixer_4]}
 {id=Random_0, children [WavePlayer_0,WavePlayer_1,WavePlayer_2], properties {weights [1,1,1]}}
 {id=Random_2, children [WavePlayer_4,WavePlayer_8,WavePlayer_9], properties {weights [1,1,1]}}
 {id=Random_3, children [WavePlayer_12, WavePlayer_13, WavePlayer_14], properties {weights [1,1,1]}}
 {
  id=SoundClass_55
  children [Attenuation_80]
  properties {soundClassOverride=&Fort_WeaponRanged_Fire_3P_PSM}
 }
 {
  id=SoundClass_56
  children [DistanceCrossFade_25]
  properties {soundClassOverride=&Fort_WeaponRanged_Fire_Distant}
 }
 {id=WavePlayer_0, properties {soundWaveAssetPtr=/MotherGameplay/Items/ReactorGrade/Audio/SoundWaves/ReactorGrade_Fire_Medium_01}}
 {id=WavePlayer_1, properties {soundWaveAssetPtr=/MotherGameplay/Items/ReactorGrade/Audio/SoundWaves/ReactorGrade_Fire_Medium_02}}
 {id=WavePlayer_10, properties {soundWaveAssetPtr=/MotherGameplay/Items/ReactorGrade/Audio/SoundWaves/ReactorGrade_Fire_Close_02}}
 {id=WavePlayer_12, properties {soundWaveAssetPtr=/MotherGameplay/Items/ReactorGrade/Audio/SoundWaves/ReactorGrade_Fire_Distant_Mono_01}}
 {id=WavePlayer_13, properties {soundWaveAssetPtr=/MotherGameplay/Items/ReactorGrade/Audio/SoundWaves/ReactorGrade_Fire_Distant_Mono_02}}
 {id=WavePlayer_14, properties {soundWaveAssetPtr=/MotherGameplay/Items/ReactorGrade/Audio/SoundWaves/ReactorGrade_Fire_Distant_Mono_03}}
 {id=WavePlayer_15, properties {soundWaveAssetPtr=&ReactorGrade_Fire_Distant_Mono_02}}
 {id=WavePlayer_2, properties {soundWaveAssetPtr=/MotherGameplay/Items/ReactorGrade/Audio/SoundWaves/ReactorGrade_Fire_Medium_03}}
 {id=WavePlayer_3, properties {soundWaveAssetPtr=&ReactorGrade_Fire_Medium_02}}
 {id=WavePlayer_4, properties {soundWaveAssetPtr=/MotherGameplay/Items/ReactorGrade/Audio/SoundWaves/ReactorGrade_Fire_Close_01}}
 {id=WavePlayer_8, properties {soundWaveAssetPtr=&ReactorGrade_Fire_Close_02}}
 {id=WavePlayer_9, properties {soundWaveAssetPtr=/MotherGameplay/Items/ReactorGrade/Audio/SoundWaves/ReactorGrade_Fire_Close_03}}
]
properties {
 maxDistance=27500
 volumeMultiplier=1
 attenuationOverrides {distanceAlgorithm=NaturalSound, falloffDistance=4000, attenuationShapeExtents {x=1000}}
 concurrencyOverrides {resolutionRule=PreventNew}
 concurrencySet [&Weapon_Gunfire_Local_Conc, &Weapon_Gunfire_Global_Conc]
}
