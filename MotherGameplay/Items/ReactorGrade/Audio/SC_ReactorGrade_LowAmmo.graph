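# Sound cue for the low-ammo warning of the ReactorGrade item, classed as weapon foley.
# A single wave player plays the ReactorGrade_LowAmmo sound wave at reduced volume and raised pitch.
# It uses small rifle attenuation and overrides concurrency so new instances are prevented rather than cutting off playing ones.
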SoundCue /MotherGameplay/Items/ReactorGrade/Audio/SC_ReactorGrade_LowAmmo
root=WavePlayer_0
nodes [{id=WavePlayer_0, properties {soundWaveAssetPtr=/MotherGameplay/Items/ReactorGrade/Audio/SoundWaves/ReactorGrade_LowAmmo}}]
properties {
 attenuationSettings=/Game/Sounds/Attenuation/Fort_Weapon_Rifle_Sml_Att
 bOverrideConcurrency=true
 maxDistance=3000
 pitchMultiplier=1.4
 soundClassObject=&Fort_WeaponFoley
 volumeMultiplier=0.3
 attenuationOverrides {distanceAlgorithm=NaturalSound, falloffDistance=4000, attenuationShapeExtents {x=1000}}
 concurrencyOverrides {resolutionRule=PreventNew}
}
