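# First-person Sound Cue for the minimum-charge sound of the ReactorGrade item, routed to the first-person ranged weapon fire sound class.
# A single wave player feeds an Enveloper that applies a rising pitch curve and slight random pitch variation.
# Concurrency is limited per owner to a couple of instances, stopping the oldest when exceeded.
# The cue has a boosted volume multiplier and its debug flag enabled.
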SoundCue /MotherGameplay/Items/ReactorGrade/Audio/SC_ReactorGrade_MinimumCharge_1P
root=Enveloper_0
nodes [
 {
  id=Enveloper_0
  children [WavePlayer_3]
  properties {pitchMax=1.1, pitchMin=0.9, pitchCurve {editorCurveData {keys [{time=0}, <at 1 1.3>]}}, volumeCurve {editorCurveData {keys [<at 0 1>]}}}
 }
 {id=WavePlayer_3, properties {soundWaveAssetPtr=/MotherGameplay/Items/ReactorGrade/Audio/SoundWaves/ReactorGrade_MinCharge_01}}
]
properties {
 bDebug=true
 maxDistance=2097152
 soundClassObject=&Fort_WeaponRanged_Fire_1P_PSM
 volumeMultiplier=1.4
 concurrencyOverrides {bLimitToOwner=true, maxCount=2, resolutionRule=StopOldest}
}
