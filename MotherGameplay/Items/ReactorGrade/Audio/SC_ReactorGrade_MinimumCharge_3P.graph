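# Third-person sound cue for the ReactorGrade item's minimum charge, routed to the ranged weapon fire 3P sound class.
# A single wave plays through an Enveloper that adds slight random pitch variation and a rising pitch curve.
# An Attenuation node with overridden distance falloff and low-pass filtering shapes it over range.
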SoundCue /MotherGameplay/Items/ReactorGrade/Audio/SC_ReactorGrade_MinimumCharge_3P
root=Attenuation_0
nodes [
 {
  id=Attenuation_0
  children [Enveloper_0]
  properties {
   attenuationSettings=&Fort_Weapon_Projectile_Close_Athena_Att
   bOverrideAttenuation=true
   attenuationOverrides {bAttenuateWithLPF=true, falloffDistance=5000, lPFFrequencyAtMax=800, lPFRadiusMax=2000, lPFRadiusMin=500, attenuationShapeExtents {x=750}}
  }
 }
 {
  id=Enveloper_0
  children [WavePlayer_3]
  properties {pitchMax=1.1, pitchMin=0.9, pitchCurve {editorCurveData {keys [{time=0}, <at 1 1.3>]}}, volumeCurve {editorCurveData {keys [<at 0 1>]}}}
 }
 {id=WavePlayer_3, properties {soundWaveAssetPtr=/MotherGameplay/Items/ReactorGrade/Audio/SoundWaves/ReactorGrade_MinCharge_01}}
]
properties {
 bDebug=true
 maxDistance=5750
 soundClassObject=&Fort_WeaponRanged_Fire_3P_PSM
 volumeMultiplier=0.7
 concurrencyOverrides {bLimitToOwner=true, maxCount=2, resolutionRule=StopOldest}
}
