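# Sound cue for the out-of-ammo sound of the ReactorGrade item, built from reused sniper out-of-ammo recordings.
# A Random node picks among four equally weighted waves and feeds a Modulator that holds pitch constant.
# The cue lowers volume and raises overall pitch, and routes to a third-person ranged weapon fire sound class.
# A concurrency override limits it to two simultaneous instances, stopping the oldest.
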SoundCue /MotherGameplay/Items/ReactorGrade/Audio/SC_ReactorGrade_OutOfAmmo
root=Modulator_1
nodes [
 {id=Modulator_1, children [Random_1], properties {pitchMax=1, pitchMin=1}}
 {id=Random_1, children [WavePlayer_4,WavePlayer_5,WavePlayer_6,WavePlayer_7], properties {weights [1,1,1,1]}}
 {id=WavePlayer_4, properties {soundWaveAssetPtr=/MotherGameplay/Items/ReactorGrade/Audio/SoundWaves/Sniper_OutofAmmo_01}}
 {id=WavePlayer_5, properties {soundWaveAssetPtr=/MotherGameplay/Items/ReactorGrade/Audio/SoundWaves/Sniper_OutofAmmo_02}}
 {id=WavePlayer_6, properties {soundWaveAssetPtr=/MotherGameplay/Items/ReactorGrade/Audio/SoundWaves/Sniper_OutofAmmo_03}}
 {id=WavePlayer_7, properties {soundWaveAssetPtr=/MotherGameplay/Items/ReactorGrade/Audio/SoundWaves/Sniper_OutofAmmo_04}}
]
properties {
 bOverrideConcurrency=true
 maxDistance=2097152
 pitchMultiplier=1.5
 soundClassObject=&Fort_WeaponRanged_Fire_3P_PSM
 volumeMultiplier=0.2
 attenuationOverrides {falloffDistance=1600}
 concurrencyOverrides {maxCount=2, resolutionRule=StopOldest}
}
