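# Sound cue played when the ReactorGrade item is picked up, routed to the loot collect sound class with weapon pickup attenuation.
# A mixer layers the main pickup wave with a quieter, pitched-up energy variant chosen at random from four and slightly delayed.
# An outer modulator varies the boosted mix volume, while the cue itself scales volume and pitch down.
# A concurrency override allows one instance at a time and blocks new plays.
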SoundCue /MotherGameplay/Items/ReactorGrade/Audio/SC_ReactorGrade_Pickup
root=Modulator_1
nodes [
 {id=Delay_0, children [Random_0], properties {delayMax=0.2, delayMin=0.1}}
 {id=Mixer_0, children [Modulator_0,WavePlayer_5], properties {inputVolume [0.6,1]}}
 {id=Modulator_0, children [Delay_0], properties {pitchMax=1.2, pitchMin=1.1}}
 {id=Modulator_1, children [Mixer_0], properties {volumeMax=2, volumeMin=1.8}}
 {id=Random_0, children [WavePlayer_0,WavePlayer_1,WavePlayer_3,WavePlayer_4], properties {weights [1,1,1,1]}}
 {id=WavePlayer_0, properties {soundWaveAssetPtr=/MotherGameplay/Items/ReactorGrade/Audio/SoundWaves/ReactorGrade_Pickup_Energy_04}}
 {id=WavePlayer_1, properties {soundWaveAssetPtr=/MotherGameplay/Items/ReactorGrade/Audio/SoundWaves/ReactorGrade_Pickup_Energy_03}}
 {id=WavePlayer_3, properties {soundWaveAssetPtr=/MotherGameplay/Items/ReactorGrade/Audio/SoundWaves/ReactorGrade_Pickup_Energy_02}}
 {id=WavePlayer_4, properties {soundWaveAssetPtr=/MotherGameplay/Items/ReactorGrade/Audio/SoundWaves/ReactorGrade_Pickup_Energy_01}}
 {id=WavePlayer_5, properties {soundWaveAssetPtr=/MotherGameplay/Items/ReactorGrade/Audio/SoundWaves/ReactorGrade_Pickup}}
]
properties {
 attenuationSettings=&Fort_Weapon_Pickup_Att
 bOverrideConcurrency=true
 maxDistance=3500
 pitchMultiplier=0.8
 soundClassObject=&Fort_Loot_Collect
 volumeMultiplier=0.15
 concurrencyOverrides {maxCount=1, resolutionRule=PreventNew}
}
