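# Sound cue for the ReactorGrade projectile hitting dirt, using bullet-impact attenuation and rifle impact concurrency.
# A raised, randomized pitch wraps a quality switch in which each branch layers a random projectile impact with a delayed rock impact.
# The first two quality slots mix two staggered random rock layers from larger pools.
# The last two slots use a single two-variant rock layer, with their randoms excluded from branch culling.
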SoundCue /MotherGameplay/Items/ReactorGrade/Audio/SC_ReactorGrade_Projectile_DirtImpact
root=Modulator_0
nodes [
 {id=Delay_0, children [Mixer_3], properties {delayMax=0.2, delayMin=0.2}}
 {id=Delay_1, children [Random_4], properties {delayMax=0.2, delayMin=0.2}}
 {id=Delay_2, children [Modulator_2], properties {delayMax=0.1, delayMin=0.05}}
 {id=Mixer_1, children [Random_0,Delay_0], properties {inputVolume [1,1.4]}}
 {id=Mixer_2, children [Random_1,Delay_1], properties {inputVolume [1,1.4]}}
 {id=Mixer_3, children [Modulator_3,Delay_2], properties {inputVolume [1,0.2]}}
 {id=Modulator_0, children [QualityLevel_0], properties {pitchMax=1.4, pitchMin=1.1}}
 {id=Modulator_2, children [Random_3]}
 {id=Modulator_3, children [Random_7]}
 {id=QualityLevel_0, children [Mixer_1,Mixer_1,Mixer_2,Mixer_2]}
 {id=Random_0, children [WavePlayer_1,WavePlayer_3], properties {weights [1,1]}}
 {id=Random_1, children [WavePlayer_5,WavePlayer_4], properties {bShouldExcludeFromBranchCulling=true, weights [1,1]}}
 {id=Random_3, children [WavePlayer_119, WavePlayer_122, WavePlayer_123, WavePlayer_117], properties {weights [1,1,1,1]}}
 {id=Random_4, children [WavePlayer_7,WavePlayer_8], properties {bShouldExcludeFromBranchCulling=true, weights [1,1]}}
 {id=Random_7, children [WavePlayer_109, WavePlayer_108, WavePlayer_115], properties {weights [1,1,1]}}
 {id=WavePlayer_1, properties {soundWaveAssetPtr=/MotherGameplay/Items/ReactorGrade/Audio/SoundWaves/ReactorGrade_Projectile_Impact_02}}
 {id=WavePlayer_108, properties {soundWaveAssetPtr=/Game/Sounds/Fort_Impact_Sounds/Demolition/Defaults/rock_impact_05}}
 {id=WavePlayer_109, properties {soundWaveAssetPtr=/Game/Sounds/Fort_Impact_Sounds/Demolition/Defaults/rock_impact_04}}
 {id=WavePlayer_115, properties {soundWaveAssetPtr=/Game/Sounds/Fort_Impact_Sounds/Demolition/Defaults/rock_impact_07}}
 {id=WavePlayer_117, properties {soundWaveAssetPtr=/Game/Sounds/Fort_Impact_Sounds/Demolition/Defaults/rock_impact_stereo_01}}
 {id=WavePlayer_119, properties {soundWaveAssetPtr=/Game/Sounds/Fort_Impact_Sounds/Demolition/Defaults/rock_impact_stereo_03}}
 {id=WavePlayer_122, properties {soundWaveAssetPtr=/Game/Sounds/Fort_Impact_Sounds/Demolition/Defaults/rock_impact_stereo_06}}
 {id=WavePlayer_123, properties {soundWaveAssetPtr=/Game/Sounds/Fort_Impact_Sounds/Demolition/Defaults/rock_impact_stereo_07}}
 {id=WavePlayer_3, properties {soundWaveAssetPtr=/MotherGameplay/Items/ReactorGrade/Audio/SoundWaves/ReactorGrade_Projectile_Impact_04}}
 {id=WavePlayer_4, properties {soundWaveAssetPtr=&ReactorGrade_Projectile_Impact_04}}
 {id=WavePlayer_5, properties {soundWaveAssetPtr=&ReactorGrade_Projectile_Impact_02}}
 {id=WavePlayer_7, properties {soundWaveAssetPtr=&rock_impact_stereo_03}}
 {id=WavePlayer_8, properties {soundWaveAssetPtr=&rock_impact_stereo_07}}
]
properties {
 attenuationSettings=/Game/Sounds/Attenuation/Fort_Bullet_Impact_Att
 maxDistance=5300
 soundClassObject=&Fort_Impacts
 volumeMultiplier=1
 attenuationOverrides {falloffDistance=5000, attenuationShapeExtents {x=1000}}
 concurrencyOverrides {maxCount=4, volumeScale=0.8}
 concurrencySet [&RifleBulletImpact_Conc]
}
