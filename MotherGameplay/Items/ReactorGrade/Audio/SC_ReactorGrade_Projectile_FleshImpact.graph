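# Sound cue for a Reactor Grade projectile striking flesh, played through the impacts sound class with bullet impact attenuation and rifle bullet impact concurrency.
# A root modulator randomizes pitch upward and feeds a quality-level switch.
# The first two quality slots share a random pick of three impact waves.
# The last two share a pick of two of those waves, excluded from branch culling.
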SoundCue /MotherGameplay/Items/ReactorGrade/Audio/SC_ReactorGrade_Projectile_FleshImpact
root=Modulator_0
nodes [
 {id=Modulator_0, children [QualityLevel_0], properties {pitchMax=1.4, pitchMin=1.1}}
 {id=QualityLevel_0, children [Random_0,Random_0,Random_1,Random_1]}
 {id=Random_0, children [WavePlayer_0,WavePlayer_1,WavePlayer_3], properties {weights [1,1,1]}}
 {id=Random_1, children [WavePlayer_5,WavePlayer_4], properties {bShouldExcludeFromBranchCulling=true, weights [1,1]}}
 {id=WavePlayer_0, properties {soundWaveAssetPtr=/MotherGameplay/Items/ReactorGrade/Audio/SoundWaves/ReactorGrade_Projectile_Impact_01}}
 {id=WavePlayer_1, properties {soundWaveAssetPtr=/MotherGameplay/Items/ReactorGrade/Audio/SoundWaves/ReactorGrade_Projectile_Impact_02}}
 {id=WavePlayer_3, properties {soundWaveAssetPtr=/MotherGameplay/Items/ReactorGrade/Audio/SoundWaves/ReactorGrade_Projectile_Impact_04}}
 {id=WavePlayer_4, properties {soundWaveAssetPtr=&ReactorGrade_Projectile_Impact_04}}
 {id=WavePlayer_5, properties {soundWaveAssetPtr=&ReactorGrade_Projectile_Impact_02}}
]
properties {
 attenuationSettings=/Game/Sounds/Attenuation/Fort_Bullet_Impact_Att
 maxDistance=5300
 soundClassObject=&Fort_Impacts
 volumeMultiplier=1
 attenuationOverrides {falloffDistance=5000, attenuationShapeExtents {x=1000}}
 concurrencyOverrides {maxCount=4, volumeScale=0.8}
 concurrencySet [&RifleBulletImpact_Conc]
}
