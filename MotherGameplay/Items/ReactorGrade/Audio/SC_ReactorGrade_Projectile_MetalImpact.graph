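# Sound cue for a ReactorGrade projectile hitting metal, classed as an impact sound with bullet-impact attenuation and rifle-impact concurrency.
# A root modulator randomly raises pitch over a quality switch that shares two mixes across four levels.
# The fuller mix plays a random projectile impact, then delayed metal layers (final, distant, further-delayed close).
# The leaner mix uses fewer variations and one delayed close-metal layer, with its randoms excluded from branch culling.
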SoundCue /MotherGameplay/Items/ReactorGrade/Audio/SC_ReactorGrade_Projectile_MetalImpact
root=Modulator_0
nodes [
 {id=Delay_0, children [Mixer_7], properties {delayMax=0.2, delayMin=0.2}}
 {id=Delay_1, children [Random_6], properties {delayMax=0.2, delayMin=0.2}}
 {id=Delay_6, children [Modulator_1], properties {delayMax=0.2, delayMin=0.1}}
 {id=Mixer_1, children [Random_0,Delay_0], properties {inputVolume [1,0.8]}}
 {id=Mixer_2, children [Random_1,Delay_1], properties {inputVolume [1,0.8]}}
 {id=Mixer_7, children [Modulator_6,Delay_6,Random_2], properties {inputVolume [1,1,1]}}
 {id=Modulator_0, children [QualityLevel_0], properties {pitchMax=1.4, pitchMin=1.1}}
 {id=Modulator_1, children [Random_18]}
 {id=Modulator_6, children [Random_15]}
 {id=QualityLevel_0, children [Mixer_1,Mixer_1,Mixer_2,Mixer_2]}
 {id=Random_0, children [WavePlayer_0,WavePlayer_1,WavePlayer_3], properties {weights [1,1,1]}}
 {id=Random_1, children [WavePlayer_5,WavePlayer_4], properties {bShouldExcludeFromBranchCulling=true, weights [1,1]}}
 {id=Random_15, children [WavePlayer_43, WavePlayer_44], properties {weights [1,1]}}
 {id=Random_18, children [WavePlayer_46, WavePlayer_47, WavePlayer_48], properties {weights [1,1,1]}}
 {id=Random_2, children [WavePlayer_2, WavePlayer_6, WavePlayer_11], properties {weights [1,1,1]}}
 {id=Random_6, children [WavePlayer_12, WavePlayer_13], properties {bShouldExcludeFromBranchCulling=true, weights [1,1]}}
 {id=WavePlayer_0, properties {soundWaveAssetPtr=/MotherGameplay/Items/ReactorGrade/Audio/SoundWaves/ReactorGrade_Projectile_Impact_01}}
 {id=WavePlayer_1, properties {soundWaveAssetPtr=/MotherGameplay/Items/ReactorGrade/Audio/SoundWaves/ReactorGrade_Projectile_Impact_02}}
 {id=WavePlayer_11, properties {soundWaveAssetPtr=/Game/Sounds/Fort_Impact_Sounds/Demolition/Defaults/Metal_Distant_Impact_04}}
 {id=WavePlayer_12, properties {soundWaveAssetPtr=/Game/Sounds/Fort_Impact_Sounds/Demolition/Defaults/metal_close_impact_01}}
 {id=WavePlayer_13, properties {soundWaveAssetPtr=/Game/Sounds/Fort_Impact_Sounds/Demolition/Defaults/metal_close_impact_02}}
 {id=WavePlayer_2, properties {soundWaveAssetPtr=/Game/Sounds/Fort_Impact_Sounds/Demolition/Defaults/Metal_Distant_Impact_01}}
 {id=WavePlayer_3, properties {soundWaveAssetPtr=/MotherGameplay/Items/ReactorGrade/Audio/SoundWaves/ReactorGrade_Projectile_Impact_04}}
 {id=WavePlayer_4, properties {soundWaveAssetPtr=&ReactorGrade_Projectile_Impact_04}}
 {id=WavePlayer_43, properties {soundWaveAssetPtr=/Game/Sounds/Fort_Impact_Sounds/Demolition/Defaults/final_metal_new_01}}
 {id=WavePlayer_44, properties {soundWaveAssetPtr=/Game/Sounds/Fort_Impact_Sounds/Demolition/Defaults/final_metal_new_02}}
 {id=WavePlayer_46, properties {soundWaveAssetPtr=&metal_close_impact_01}}
 {id=WavePlayer_47, properties {soundWaveAssetPtr=&metal_close_impact_02}}
 {id=WavePlayer_48, properties {soundWaveAssetPtr=/Game/Sounds/Fort_Impact_Sounds/Demolition/Defaults/metal_close_impact_03}}
 {id=WavePlayer_5, properties {soundWaveAssetPtr=&ReactorGrade_Projectile_Impact_02}}
 {id=WavePlayer_6, properties {soundWaveAssetPtr=/Game/Sounds/Fort_Impact_Sounds/Demolition/Defaults/Metal_Distant_Impact_03}}
]
properties {
 attenuationSettings=/Game/Sounds/Attenuation/Fort_Bullet_Impact_Att
 maxDistance=5300
 soundClassObject=&Fort_Impacts
 volumeMultiplier=1
 attenuationOverrides {falloffDistance=5000, attenuationShapeExtents {x=1000}}
 concurrencyOverrides {maxCount=4, volumeScale=0.8}
 concurrencySet [&RifleBulletImpact_Conc]
}
