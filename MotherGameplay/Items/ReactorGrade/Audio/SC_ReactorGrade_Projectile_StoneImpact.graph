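# Sound cue for the ReactorGrade projectile hitting stone, attenuated as a bullet impact with limited concurrency.
# A top-level random pitch shift feeds a quality switch, with each branch mixing a random projectile impact and delayed rock debris.
# The first two quality slots layer four pitch-varied, staggered rock hits for a tumbling tail.
# The other two use a single delayed rock hit, with random nodes excluded from branch culling.
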SoundCue /MotherGameplay/Items/ReactorGrade/Audio/SC_ReactorGrade_Projectile_StoneImpact
root=Modulator_0
nodes [
 {id=Delay_0, children [Mixer_3], properties {delayMax=0.2, delayMin=0.2}}
 {id=Delay_1, children [Random_4], properties {delayMax=0.2, delayMin=0.2}}
 {id=Delay_2, children [Modulator_2], properties {delayMax=0.2, delayMin=0.15}}
 {id=Delay_4, children [Modulator_10], properties {delayMax=0.08, delayMin=0.05}}
 {id=Delay_5, children [Modulator_11], properties {delayMax=0.52, delayMin=0.45}}
 {id=Mixer_1, children [Random_0,Delay_0], properties {inputVolume [1,0.8]}}
 {id=Mixer_2, children [Random_1,Delay_1], properties {inputVolume [1,0.8]}}
 {id=Mixer_3, children [Modulator_3,Delay_2,Delay_4,Delay_5], properties {inputVolume [1,0.7,0.5,0.4]}}
 {id=Modulator_0, children [QualityLevel_0], properties {pitchMax=1.4, pitchMin=1.1}}
 {id=Modulator_10, children [Random_5], properties {pitchMax=1.5, pitchMin=1.3}}
 {id=Modulator_11, children [Random_5], properties {pitchMax=1.5, pitchMin=1.3}}
 {id=Modulator_2, children [Random_10], properties {pitchMax=1.5, pitchMin=1.4}}
 {id=Modulator_3, children [Random_3], properties {pitchMax=1.2}}
 {id=QualityLevel_0, children [Mixer_1,Mixer_1,Mixer_2,Mixer_2]}
 {id=Random_0, children [WavePlayer_0,WavePlayer_1,WavePlayer_3], properties {weights [1,1,1]}}
 {id=Random_1, children [WavePlayer_5,WavePlayer_4], properties {bShouldExcludeFromBranchCulling=true, weights [1,1]}}
 {id=Random_10, children [WavePlayer_55, WavePlayer_58, WavePlayer_59], properties {weights [1,1,1]}}
 {id=Random_3, children [WavePlayer_8,WavePlayer_7], properties {weights [1,1]}}
 {id=Random_4, children [WavePlayer_10, WavePlayer_9], properties {bShouldExcludeFromBranchCulling=true, weights [1,1]}}
 {id=Random_5, children [WavePlayer_63, WavePlayer_64, WavePlayer_65], properties {weights [1,1,1]}}
 {id=WavePlayer_0, properties {soundWaveAssetPtr=/MotherGameplay/Items/ReactorGrade/Audio/SoundWaves/ReactorGrade_Projectile_Impact_01}}
 {id=WavePlayer_1, properties {soundWaveAssetPtr=/MotherGameplay/Items/ReactorGrade/Audio/SoundWaves/ReactorGrade_Projectile_Impact_02}}
 {id=WavePlayer_10, properties {soundWaveAssetPtr=/Game/Sounds/Fort_Impact_Sounds/Demolition/Defaults/rock_impact_stereo_08}}
 {id=WavePlayer_3, properties {soundWaveAssetPtr=/MotherGameplay/Items/ReactorGrade/Audio/SoundWaves/ReactorGrade_Projectile_Impact_04}}
 {id=WavePlayer_4, properties {soundWaveAssetPtr=&ReactorGrade_Projectile_Impact_04}}
 {id=WavePlayer_5, properties {soundWaveAssetPtr=&ReactorGrade_Projectile_Impact_02}}
 {id=WavePlayer_55, properties {soundWaveAssetPtr=/Game/Sounds/Fort_Impact_Sounds/Demolition/Defaults/rock_impact_stereo_03}}
 {id=WavePlayer_58, properties {soundWaveAssetPtr=/Game/Sounds/Fort_Impact_Sounds/Demolition/Defaults/rock_impact_stereo_06}}
 {id=WavePlayer_59, properties {soundWaveAssetPtr=/Game/Sounds/Fort_Impact_Sounds/Demolition/Defaults/rock_impact_stereo_07}}
 {id=WavePlayer_63, properties {soundWaveAssetPtr=/Game/Sounds/Fort_Impact_Sounds/Demolition/Defaults/Fort_Impact_RockFinal_01}}
 {id=WavePlayer_64, properties {soundWaveAssetPtr=/Game/Sounds/Fort_Impact_Sounds/Demolition/Defaults/Fort_Impact_RockFinal_02}}
 {id=WavePlayer_65, properties {soundWaveAssetPtr=/Game/Sounds/Fort_Impact_Sounds/Demolition/Defaults/Fort_Impact_RockFinal_03}}
 {id=WavePlayer_7, properties {soundWaveAssetPtr=/Game/Sounds/Fort_Impact_Sounds/Demolition/Defaults/rock_impact_stereo_10}}
 {id=WavePlayer_8, properties {soundWaveAssetPtr=&rock_impact_stereo_08}}
 {id=WavePlayer_9, properties {soundWaveAssetPtr=&rock_impact_stereo_10}}
]
properties {
 attenuationSettings=/Game/Sounds/Attenuation/Fort_Bullet_Impact_Att
 maxDistance=5300
 soundClassObject=&Fort_Impacts
 volumeMultiplier=1
 attenuationOverrides {falloffDistance=5000, attenuationShapeExtents {x=1000}}
 concurrencyOverrides {maxCount=4, volumeScale=0.8}
 concurrencySet [&RifleBulletImpact_Conc]
}
