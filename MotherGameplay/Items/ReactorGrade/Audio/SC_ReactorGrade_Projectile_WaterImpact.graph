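# Sound cue for a ReactorGrade projectile hitting water, using bullet impact attenuation and rifle bullet impact concurrency.
# Each mix layers a random projectile impact with a louder, briefly delayed random swim-jump splash.
# A quality switch sends the first two levels to a three-splash mix and the last two to a two-splash mix with randoms excluded from branch culling.
# A top-level modulator randomly raises the pitch.
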SoundCue /MotherGameplay/Items/ReactorGrade/Audio/SC_ReactorGrade_Projectile_WaterImpact
root=Modulator_0
nodes [
 {id=Delay_0, children [Modulator_1], properties {delayMax=0.05, delayMin=0.05}}
 {id=Delay_1, children [Random_6], properties {delayMax=0.05, delayMin=0.05}}
 {id=Mixer_1, children [Random_0,Delay_0], properties {inputVolume [0.85,1.5]}}
 {id=Mixer_2, children [Random_1,Delay_1], properties {inputVolume [0.85,1.5]}}
 {id=Modulator_0, children [QualityLevel_0], properties {pitchMax=1.4, pitchMin=1.1}}
 {id=Modulator_1, children [Random_5], properties {volumeMax=1, volumeMin=1}}
 {id=QualityLevel_0, children [Mixer_1,Mixer_1,Mixer_2,Mixer_2]}
 {id=Random_0, children [WavePlayer_1,WavePlayer_3], properties {weights [1,1]}}
 {id=Random_1, children [WavePlayer_5,WavePlayer_4], properties {bShouldExcludeFromBranchCulling=true, weights [1,1]}}
 {id=Random_5, children [WavePlayer_14, WavePlayer_16, WavePlayer_17], properties {weights [1,1,1]}}
 {id=Random_6, children [WavePlayer_18, WavePlayer_21], properties {bShouldExcludeFromBranchCulling=true, weights [1,1]}}
 {id=WavePlayer_1, properties {soundWaveAssetPtr=/MotherGameplay/Items/ReactorGrade/Audio/SoundWaves/ReactorGrade_Projectile_Impact_02}}
 {id=WavePlayer_14, properties {soundWaveAssetPtr=/Game/Athena/Sounds/Water/Interaction/Water_Player_SwimJump_1P_01}}
 {id=WavePlayer_16, properties {soundWaveAssetPtr=/Game/Athena/Sounds/Water/Interaction/Water_Player_SwimJump_1P_03}}
 {id=WavePlayer_17, properties {soundWaveAssetPtr=/Game/Athena/Sounds/Water/Interaction/Water_Player_SwimJump_1P_04}}
 {id=WavePlayer_18, properties {soundWaveAssetPtr=&Water_Player_SwimJump_1P_03}}
 {id=WavePlayer_21, properties {soundWaveAssetPtr=&Water_Player_SwimJump_1P_04}}
 {id=WavePlayer_3, properties {soundWaveAssetPtr=/MotherGameplay/Items/ReactorGrade/Audio/SoundWaves/ReactorGrade_Projectile_Impact_04}}
 {id=WavePlayer_4, properties {soundWaveAssetPtr=&ReactorGrade_Projectile_Impact_04}}
 {id=WavePlayer_5, properties {soundWaveAssetPtr=&ReactorGrade_Projectile_Impact_02}}
]
properties {
 attenuationSettings=/Game/Sounds/Attenuation/Fort_Bullet_Impact_Att
 maxDistance=5300
 soundClassObject=&Fort_Impacts
 volumeMultiplier=1
 attenuationOverrides {falloffDistance=5000, attenuationShapeExtents {x=1000}}
 concurrencyOverrides {maxCount=4, volumeScale=0.8}
 concurrencySet [&RifleBulletImpact_Conc]
}
