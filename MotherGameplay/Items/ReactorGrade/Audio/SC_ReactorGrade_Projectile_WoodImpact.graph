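# Sound cue for a ReactorGrade projectile hitting wood, using bullet-impact attenuation and rifle-impact concurrency.
# A pitch-raised quality switch gives the first two levels a randomized projectile impact, followed by a delayed mix of wood impact and two staggered, pitch-varied shatters.
# The last two levels use a lighter mix with fewer variants and only a delayed wood impact, with those random nodes excluded from branch culling.
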SoundCue /MotherGameplay/Items/ReactorGrade/Audio/SC_ReactorGrade_Projectile_WoodImpact
root=Modulator_0
nodes [
 {id=Delay_0, children [Modulator_1], properties {delayMax=0.2, delayMin=0.2}}
 {id=Delay_1, children [Random_2], properties {delayMax=0.2, delayMin=0.2}}
 {id=Delay_6, children [Modulator_25], properties {delayMax=0.12, delayMin=0.1}}
 {id=Mixer_1, children [Random_0,Delay_0], properties {inputVolume [1,0.8]}}
 {id=Mixer_2, children [Random_1,Delay_1], properties {inputVolume [1,0.8]}}
 {id=Mixer_7, children [Modulator_24,Delay_6,Random_40], properties {inputVolume [1,1,0.7]}}
 {id=Modulator_0, children [QualityLevel_0], properties {pitchMax=1.4, pitchMin=1.1}}
 {id=Modulator_1, children [Mixer_7]}
 {id=Modulator_24, children [Random_42], properties {pitchMax=1.5, pitchMin=1.2}}
 {id=Modulator_25, children [Random_41], properties {pitchMax=1.4, pitchMin=1.2}}
 {id=QualityLevel_0, children [Mixer_1,Mixer_1,Mixer_2,Mixer_2]}
 {id=Random_0, children [WavePlayer_0,WavePlayer_1,WavePlayer_3], properties {weights [1,1,1]}}
 {id=Random_1, children [WavePlayer_5,WavePlayer_4], properties {bShouldExcludeFromBranchCulling=true, weights [1,1]}}
 {id=Random_2, children [WavePlayer_6,WavePlayer_2], properties {bShouldExcludeFromBranchCulling=true, weights [1,1]}}
 {id=Random_40, children [WavePlayer_138, WavePlayer_137], properties {weights [1,1]}}
 {id=Random_41, children [WavePlayer_141, WavePlayer_142, WavePlayer_143], properties {weights [1,1,1]}}
 {id=Random_42, children [WavePlayer_139, WavePlayer_140], properties {weights [1,1]}}
 {id=WavePlayer_0, properties {soundWaveAssetPtr=/MotherGameplay/Items/ReactorGrade/Audio/SoundWaves/ReactorGrade_Projectile_Impact_01}}
 {id=WavePlayer_1, properties {soundWaveAssetPtr=/MotherGameplay/Items/ReactorGrade/Audio/SoundWaves/ReactorGrade_Projectile_Impact_02}}
 {id=WavePlayer_137, properties {soundWaveAssetPtr=/Game/Sounds/Fort_Impact_Sounds/Demolition/Defaults/wood_impact_stereo_04}}
 {id=WavePlayer_138, properties {soundWaveAssetPtr=/Game/Sounds/Fort_Impact_Sounds/Demolition/Defaults/wood_impact_stereo_01}}
 {id=WavePlayer_139, properties {soundWaveAssetPtr=/Game/Sounds/Fort_Impact_Sounds/Demolition/Wood/Wood_Small_Shatter01}}
 {id=WavePlayer_140, properties {soundWaveAssetPtr=/Game/Sounds/Fort_Impact_Sounds/Demolition/Wood/Wood_Small_Shatter02}}
 {id=WavePlayer_141, properties {soundWaveAssetPtr=/Game/Sounds/Fort_Impact_Sounds/Demolition/Wood/Wood_Small_Shatter03}}
 {id=WavePlayer_142, properties {soundWaveAssetPtr=/Game/Sounds/Fort_Impact_Sounds/Demolition/Wood/Wood_Small_Shatter04}}
 {id=WavePlayer_143, properties {soundWaveAssetPtr=/Game/Sounds/Fort_Impact_Sounds/Demolition/Wood/Wood_Small_Shatter05}}
 {id=WavePlayer_2, properties {soundWaveAssetPtr=&wood_impact_stereo_04}}
 {id=WavePlayer_3, properties {soundWaveAssetPtr=/MotherGameplay/Items/ReactorGrade/Audio/SoundWaves/ReactorGrade_Projectile_Impact_04}}
 {id=WavePlayer_4, properties {soundWaveAssetPtr=&ReactorGrade_Projectile_Impact_04}}
 {id=WavePlayer_5, properties {soundWaveAssetPtr=&ReactorGrade_Projectile_Impact_02}}
 {id=WavePlayer_6, properties {soundWaveAssetPtr=&wood_impact_stereo_01}}
]
properties {
 attenuationSettings=/Game/Sounds/Attenuation/Fort_Bullet_Impact_Att
 maxDistance=5300
 soundClassObject=&Fort_Impacts
 volumeMultiplier=1
 attenuationOverrides {falloffDistance=5000, attenuationShapeExtents {x=1000}}
 concurrencyOverrides {maxCount=4, volumeScale=0.8}
 concurrencySet [&RifleBulletImpact_Conc]
}
