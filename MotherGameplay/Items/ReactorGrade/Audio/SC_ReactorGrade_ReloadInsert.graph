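# Sound cue for the reload insert sound of the ReactorGrade item.
# A single wave player feeds one insert wave straight to the output, with no randomization or modulation.
# It uses the shared weapon reload attenuation and weapon foley sound class, with reduced volume, a concurrency override and the debug flag enabled.
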SoundCue /MotherGameplay/Items/ReactorGrade/Audio/SC_ReactorGrade_ReloadInsert
root=WavePlayer_3
nodes [{id=WavePlayer_3, properties {soundWaveAssetPtr=/MotherGameplay/Items/ReactorGrade/Audio/SoundWaves/ReactorGrade_Reload_Insert_01}}]
properties {
 attenuationSettings=&Fort_Weapon_Reload_Att
 bDebug=true
 bOverrideConcurrency=true
 maxDistance=2250
 soundClassObject=&Fort_WeaponFoley
 volumeMultiplier=0.7
}
