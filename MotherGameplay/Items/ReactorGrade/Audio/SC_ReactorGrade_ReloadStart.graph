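# Sound cue for the start of the ReactorGrade weapon's reload, classed as weapon foley with reload attenuation.
# A mixer plays the reload-start wave immediately and layers a quieter secondary branch over it.
# That branch waits just under a second, then plays one of three randomly chosen military sniper casing sounds, pitched down by a modulator.
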SoundCue /MotherGameplay/Items/ReactorGrade/Audio/SC_ReactorGrade_ReloadStart
root=Mixer_0
nodes [
 {id=Delay_0, children [Random_4], properties {delayMax=0.9, delayMin=0.8}}
 {id=Mixer_0, children [WavePlayer_1,Modulator_2], properties {inputVolume [1,0.4]}}
 {id=Modulator_2, children [Delay_0], properties {pitchMax=0.7, pitchMin=0.6}}
 {id=Random_4, children [WavePlayer_16, WavePlayer_17, WavePlayer_18], properties {weights [1,1,1]}}
 {id=WavePlayer_1, properties {soundWaveAssetPtr=/MotherGameplay/Items/ReactorGrade/Audio/SoundWaves/ReactorGrade_Reload_Start_01}}
 {id=WavePlayer_16, properties {soundWaveAssetPtr=/Game/Sounds/Fort_Weapons/SniperRifles/MilitarySniper/MilitarySniper_Casing_01}}
 {id=WavePlayer_17, properties {soundWaveAssetPtr=/Game/Sounds/Fort_Weapons/SniperRifles/MilitarySniper/MilitarySniper_Casing_02}}
 {id=WavePlayer_18, properties {soundWaveAssetPtr=/Game/Sounds/Fort_Weapons/SniperRifles/MilitarySniper/MilitarySniper_Casing_03}}
]
properties {
 attenuationSettings=&Fort_Weapon_Reload_Att
 bDebug=true
 bOverrideConcurrency=true
 maxDistance=2250
 soundClassObject=&Fort_WeaponFoley
 volumeMultiplier=0.4
}
