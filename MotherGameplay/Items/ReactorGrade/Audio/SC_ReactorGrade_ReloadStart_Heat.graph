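# Sound cue for the heat layer at the start of the ReactorGrade weapon's reload.
# A Random node picks one of two heat wave variations with equal weight, and a Modulator raises its pitch by a fixed amount at constant volume.
# It plays as weapon foley with reload attenuation, reduced volume and a concurrency override, and has its debug flag enabled.
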SoundCue /MotherGameplay/Items/ReactorGrade/Audio/SC_ReactorGrade_ReloadStart_Heat
root=Modulator_0
nodes [
 {id=Modulator_0, children [Random_0], properties {pitchMax=1.25, pitchMin=1.25, volumeMax=1, volumeMin=1}}
 {id=Random_0, children [WavePlayer_0,WavePlayer_2], properties {weights [1,1]}}
 {id=WavePlayer_0, properties {soundWaveAssetPtr=/MotherGameplay/Items/ReactorGrade/Audio/SoundWaves/ReactorGrade_Reload_Start_Heat_02}}
 {id=WavePlayer_2, properties {soundWaveAssetPtr=/MotherGameplay/Items/ReactorGrade/Audio/SoundWaves/ReactorGrade_Reload_Start_Heat_01}}
]
properties {
 attenuationSettings=&Fort_Weapon_Reload_Att
 bDebug=true
 bOverrideConcurrency=true
 maxDistance=2250
 soundClassObject=&Fort_WeaponFoley
 volumeMultiplier=0.6
}
