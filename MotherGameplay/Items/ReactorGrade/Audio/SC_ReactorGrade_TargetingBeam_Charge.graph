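# Sound cue for the charge-up of the ReactorGrade targeting beam, played at half volume in the third-person ranged weapon fire sound class.
# A Random node picks one of three equally weighted close charge-up waves.
# It references the targeting beam attenuation asset, with override values for distance falloff and low-pass filtering also present.
# Overridden concurrency caps it at two instances per owner, stopping the oldest.
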SoundCue /MotherGameplay/Items/ReactorGrade/Audio/SC_ReactorGrade_TargetingBeam_Charge
root=Random_0
nodes [
 {id=Random_0, children [WavePlayer_0,WavePlayer_1,WavePlayer_2], properties {weights [1,1,1]}}
 {id=WavePlayer_0, properties {soundWaveAssetPtr=/MotherGameplay/Items/ReactorGrade/Audio/SoundWaves/ReactorGrade_ChargeUp_Close_01}}
 {id=WavePlayer_1, properties {soundWaveAssetPtr=/MotherGameplay/Items/ReactorGrade/Audio/SoundWaves/ReactorGrade_ChargeUp_Close_02}}
 {id=WavePlayer_2, properties {soundWaveAssetPtr=/MotherGameplay/Items/ReactorGrade/Audio/SoundWaves/ReactorGrade_ChargeUp_Close_03}}
]
properties {
 attenuationSettings=&/ATT_ReactorGrade_TargetingBeam
 bDebug=true
 bOverrideConcurrency=true
 maxDistance=5750
 soundClassObject=&Fort_WeaponRanged_Fire_3P_PSM
 volumeMultiplier=0.5
 attenuationOverrides {
  bAttenuateWithLPF=true
  falloffDistance=5000
  lPFFrequencyAtMax=800
  lPFRadiusMax=2000
  lPFRadiusMin=500
  nonSpatializedRadiusEnd=1000
  attenuationShapeExtents {x=750}
 }
 concurrencyOverrides {bLimitToOwner=true, maxCount=2, resolutionRule=StopOldest}
}
