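# Sound cue for the whoosh played after the ReactorGrade targeting beam fires, built from projectile fly-by recordings.
# A random node picks one of two fly-by waves with equal weight, a modulator varies its pitch upward, and a sound class node routes it to the third-person ranged weapon fire class.
# The cue raises volume and pitch, uses whoosh-by attenuation, and limits concurrency per owner, stopping the oldest instance.
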SoundCue /MotherGameplay/Items/ReactorGrade/Audio/SC_ReactorGrade_TargetingBeam_PostFire
root=SoundClass_5
nodes [
 {id=Modulator_0, children [Random_0], properties {pitchMax=1.15, pitchMin=1}}
 {id=Random_0, children [WavePlayer_0,WavePlayer_1], properties {weights [1,1]}}
 {id=SoundClass_5, children [Modulator_0], properties {soundClassOverride=&Fort_WeaponRanged_Fire_3P_PSM}}
 {id=WavePlayer_0, properties {soundWaveAssetPtr=/MotherGameplay/Items/ReactorGrade/Audio/SoundWaves/ReactorGrade_Projectile_FlyBy_01}}
 {id=WavePlayer_1, properties {soundWaveAssetPtr=/MotherGameplay/Items/ReactorGrade/Audio/SoundWaves/ReactorGrade_Projectile_FlyBy_02}}
]
properties {
 attenuationSettings=&/ATT_ReactorGrade_Whoosh_By
 bDebug=true
 bOverrideConcurrency=true
 maxDistance=3072
 pitchMultiplier=1.5
 volumeMultiplier=2
 concurrencyOverrides {bLimitToOwner=true, maxCount=2, resolutionRule=StopOldest}
}
