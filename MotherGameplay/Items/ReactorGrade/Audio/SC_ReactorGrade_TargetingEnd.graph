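# Sound cue for the end of targeting on the ReactorGrade item, routed to the weapon foley sound class.
# A mixer layers a boosted UI sound with one of two equally weighted, randomly chosen targeting-end variants, which pass through a modulator.
# The cue plays at a very low overall volume, with its own attenuation and a concurrency rule that prevents new instances once the limit is reached.
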SoundCue /MotherGameplay/Items/ReactorGrade/Audio/SC_ReactorGrade_TargetingEnd
root=Mixer_0
nodes [
 {id=Mixer_0, children [Modulator_0,WavePlayer_0], properties {inputVolume [1,2.55]}}
 {id=Modulator_0, children [Random_0]}
 {id=Random_0, children [WavePlayer_1,WavePlayer_2], properties {weights [1,1]}}
 {id=WavePlayer_0, properties {soundWaveAssetPtr=/MotherGameplay/Items/ReactorGrade/Audio/SoundWaves/ReactorGrade_Targeting_End_UI_01}}
 {id=WavePlayer_1, properties {soundWaveAssetPtr=/MotherGameplay/Items/ReactorGrade/Audio/SoundWaves/ReactorGrade_Targeting_End_01}}
 {id=WavePlayer_2, properties {soundWaveAssetPtr=/MotherGameplay/Items/ReactorGrade/Audio/SoundWaves/ReactorGrade_Targeting_End_02}}
]
properties {
 bOverrideAttenuation=true
 bOverrideConcurrency=true
 maxDistance=2700
 soundClassObject=&Fort_WeaponFoley
 volumeMultiplier=0.05
 attenuationOverrides {falloffDistance=1500, nonSpatializedRadiusStart=200, attenuationShapeExtents {x=1200}}
 concurrencyOverrides {resolutionRule=PreventNew}
}
