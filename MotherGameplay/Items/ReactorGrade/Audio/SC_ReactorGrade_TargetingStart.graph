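# Sound cue for the start of targeting with the ReactorGrade item, routed to the weapon foley sound class at a low overall volume.
# A mixer layers one of two randomly chosen, modulated start sounds with a fixed, louder-mixed UI start sound.
# The cue overrides attenuation for distance falloff with a non-spatialized near radius, and its concurrency blocks new instances rather than cutting playing ones.
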SoundCue /MotherGameplay/Items/ReactorGrade/Audio/SC_ReactorGrade_TargetingStart
root=Mixer_0
nodes [
 {id=Mixer_0, children [Modulator_0,WavePlayer_3], properties {inputVolume [1,1.8]}}
 {id=Modulator_0, children [Random_0]}
 {id=Random_0, children [WavePlayer_0,WavePlayer_1], properties {weights [1,1]}}
 {id=WavePlayer_0, properties {soundWaveAssetPtr=/MotherGameplay/Items/ReactorGrade/Audio/SoundWaves/ReactorGrade_Targeting_Start_01}}
 {id=WavePlayer_1, properties {soundWaveAssetPtr=/MotherGameplay/Items/ReactorGrade/Audio/SoundWaves/ReactorGrade_Targeting_Start_02}}
 {id=WavePlayer_3, properties {soundWaveAssetPtr=/MotherGameplay/Items/ReactorGrade/Audio/SoundWaves/ReactorGrade_Targeting_Start_UI_01}}
]
properties {
 bOverrideAttenuation=true
 bOverrideConcurrency=true
 maxDistance=2700
 soundClassObject=&Fort_WeaponFoley
 volumeMultiplier=0.075
 attenuationOverrides {falloffDistance=1500, nonSpatializedRadiusStart=200, attenuationShapeExtents {x=1200}}
 concurrencyOverrides {resolutionRule=PreventNew}
}
