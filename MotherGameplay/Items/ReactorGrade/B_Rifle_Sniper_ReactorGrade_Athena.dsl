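; Weapon Blueprint for the Reactor Grade sniper rifle, a charge-up beam weapon that can trace through a data-driven number of buildings.
; Charging plays montages and runs a recoil curve when not targeting.
; A server timer replicates the camera-derived aim offset through the ChargeAimLocation component, whose cached start and direction supply the damage trace.
; Reaching minimum charge triggers audio, shake and rumble; firing, ending charge or unequipping clears the relevant effects.
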(event EventBeginPlay
  (|CallParentFunction))

(event Custom|Play%20Charge%20Montage
  (bind _montage_chargeup_weapon (Variables|Default|GetMontage_ChargeUp_Weapon))
  (bind _montage_chargeup_player (Variables|Default|GetMontage_ChargeUp_Player))
  (Utilities|IsValid _montage_chargeup_weapon
    (:Is%20Valid
      (Animation|Montage|PlayMontage (Variables|Visual|GetWeaponMesh) _montage_chargeup_weapon)
      (Utilities|IsValid _montage_chargeup_player
        (:Is%20Valid
          (bind _asfort_player_pawn (Utilities|Casting|CastToFortPlayerPawn (Game|GetInstigator)))
          (Animation|Montage|PlayMontage (Variables|/Script/Engine.Character|GetMesh _asfort_player_pawn) _montage_chargeup_player))))))

(event EventOnChargeUp
  (CallFunction|StartRecoil)
  (CallFunction|PlayChargeMontage)
  (Networking|SwitchHasAuthority
    (:Authority
      (bind _returnvalue (Utilities|Time|SetTimerbyEvent (delegate Custom|Tick%20Charge%20Aim) 0.25 true))
      (Variables|Default|SetCharging%20Handle _returnvalue)
      (CallFunction|TickChargeAim))))

(event Custom|Tick%20Charge%20Aim
  (CallFunction|UpdateAimOffset))

(event EventOnEndCharge
  (Utilities|Time|ClearandInvalidateTimerbyHandle (Variables|Default|GetCharging%20Handle))
  (CallFunction|EndRecoil)
  (bind _refffevent (Variables|Default|GetRefFFEvent)
    (:then
      (ForceFeedback|Stop _refffevent))
    (:else)))

(event FortWeapon|EventOnLocalUnEquip
  (Utilities|Time|ClearandInvalidateTimerbyHandle (Variables|Default|GetCharging%20Handle))
  (CallFunction|EndRecoil)
  (Animation|Montage|PlayMontage (Variables|Visual|GetWeaponMesh) (Variables|Default|GetMontage_ChargeUp_Weapon) 0.0) ; FORT-447924\nResets weapon animation when changing items.\nJacob Lavender
)

(event EventOnPlayWeaponFireFX
  (params (bPersistentFire :type Boolean) (bSecondaryFire :type Boolean))
  (|CallParentFunction bPersistentFire bSecondaryFire)
  (Game|Feedback|ClientStartCameraShake :self (Fortnite|GetFortPlayerControllerfromActor (Game|GetInstigator)) :Shake "/Game/Effects/Fort_Effects/BluePrints/CameraShakes/B_EnergyWeaponFire_CameraShake.B_EnergyWeaponFire_CameraShake_C" :Scale 1.0 :PlaySpace "CameraLocal")
  (Effects|Components|ParticleSystem|SetFloatParameter (CallFunction|GetBeamNiagaraComponent) "Duration" 4.0) ; Duration of lingering beam FX after firing
  (CallFunction|EndRecoil)
  (Audio|Components|Audio|Stop (Variables|Default|GetAudio_MinChargedReachedLocal))
  (Audio|Components|Audio|Stop (Variables|Default|GetAudio_MinChargedReachedRemote)))

(event EventOnLocalClientReachedMinCharge
  (Game|Feedback|ClientStartCameraShake :self (Fortnite|GetFortPlayerControllerfromActor (Game|GetInstigator)) :Shake "/Game/Effects/Fort_Effects/Camera_Effects/BP_Camera_Shake_Pulse_Zoom_Small.BP_Camera_Shake_Pulse_Zoom_Small_C" :Scale 1.0 :PlaySpace "CameraLocal")
  (ForceFeedback|SpawnForceFeedbackAttached :ForceFeedbackEffect (Variables|Default|GetFF%20Effect%20Min%20Charge%20Reached) :AttachToComponent (Variables|Visual|GetWeaponMesh) :AttachPointName "None" :LocationType "KeepRelativeOffset" :bStopWhenAttachedToDestroyed true :bLooping false :IntensityMultiplier 1.0 :StartTime 0.0 :bAutoDestroy true)
  (if (Charging|GetIsChargingWeapon)
    (bind _returnvalue (ForceFeedback|SpawnForceFeedbackAttached :ForceFeedbackEffect (Variables|Default|GetFF%20Effect%20Min%20To%20Max%20Charge) :AttachToComponent (Variables|Visual|GetWeaponMesh) :AttachPointName "None" :LocationType "KeepRelativeOffset" :bStopWhenAttachedToDestroyed true :bLooping true :IntensityMultiplier 1.0 :StartTime 0.0 :bAutoDestroy true))
    (Variables|Default|SetRefFFEvent _returnvalue)
    (Audio|Components|Audio|Play (Variables|Default|GetAudio_MinChargedReachedLocal) 0.0)))

(event EventOnRemoteClientReachedMinCharge
  (Audio|Components|Audio|Play (Variables|Default|GetAudio_MinChargedReachedRemote) 0.0))

(event Custom|Start%20Recoil
  (bind _returnvalue (GameplayTags|HasMatchingGameplayTag(Message) (Game|GetInstigator) (Variables|Default|GetIsTargetingTag)))
  (if (not _returnvalue)
    (Default|StartRecoilCurve (Variables|Default|GetFortWeaponComponent_RecoilCurve))))

(event EventOnSetTargeting
  (params (bNewIsTargeting :type Boolean))
  (if (Charging|GetIsChargingWeapon)
    (if bNewIsTargeting
      _
      (elif (FortWeapon|IsEquipped)
        (Default|StartRecoilCurve (Variables|Default|GetFortWeaponComponent_RecoilCurve))))
    (CallFunction|EndRecoil)))

(event Custom|EndRecoil
  (Default|EndRecoilCurve (Variables|Default|GetFortWeaponComponent_RecoilCurve)))

(fn GetDamageStartAndDirection
  (bind (_outdamagestartlocation _outaimdirection) (Class|FortWeaponFxComponentChargeAimLocation|GetCachedDamageStartandAim (Variables|Default|GetFortWeaponFxComponent_ChargeAimLocation)))
  (return _outdamagestartlocation _outaimdirection))

(fn UpdateAimOffset
  (bind _returnvalue (Game|GetInstigator))
  (bind _returnvalue_2 (Fortnite|GetFortPlayerControllerfromActor _returnvalue))
  (bind _playercameramanager (Variables|/Script/Engine.PlayerController|GetPlayerCameraManager _returnvalue_2))
  (Utilities|IsValid _returnvalue_2
    (:Is%20Valid
      (Utilities|IsValid _playercameramanager
        (:Is%20Valid
          (LocalVariables|SetCameraLocation (Camera|GetCameraLocation _playercameramanager))
          (LocalVariables|SetCameraRotation (Camera|GetCameraRotation _playercameramanager))
          (LocalVariables|SetPawnLocation (Transformation|GetActorLocation _returnvalue))
          (Class|FortWeaponFxComponentChargeAimLocation|SetReplicatedAimOffset (Variables|Default|GetFortWeaponFxComponent_ChargeAimLocation) (- (+ (Math|Vector|ProjectVectorontoVector (- (LocalVariables|GetPawnLocation) (LocalVariables|GetCameraLocation)) (Math|Vector|GetForwardVector (LocalVariables|GetCameraRotation))) (LocalVariables|GetCameraLocation)) (LocalVariables|GetPawnLocation))))))))

(fn ConstructionScript
  (|CallParentFunction)
  (FortWeapon|SetTraceThroughBuildingsLimit (Math|Float|Truncate (ScalableFloat|GetValueatLevel (Variables|Default|GetRow_TraceThroughBuildingsLimit) 1.0))))
