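# Translucent deferred decal material for the ReactorGrade impact, combining a tinted scorch mark, animated emissive glow and normal detail.
# Time since Spawntime over Lifetime samples a curve atlas; a debug switch loops it instead.
# Curve channels drive emissive texture crossfade (R), scorch opacity fade (G), glow intensity (B) and normal strength (A).
# UVs scale around the centre, with normals sharing emissive tiling; optional emissive passes through DeriveHDRfromLDR.
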Material /MotherGameplay/Items/ReactorGrade/VFX/MD_Decal_Impact_ReactorGrade
graph {
 properties {
  bAutomaticallySetUsageInEditor=false
  blendMode=BLEND_Translucent
  materialDomain=MD_DeferredDecal
  pixelDepthOffsetMode=PDOM_Legacy
  usedShadingModels=Default Lit
 }
 nodes [
  n1 = Subtract(A=TextureCoordinate()) {constA=0.5, constB=0.5}
  n2 = Add(A=Multiply(A=@n1, B=ScalarParameter() {defaultValue=1, group=Emissive, parameterName=EmissiveTiling})) {constB=0.5}
  n3 = TextureSampleParameter2D(UVs=@n2) {
   group=Textures
   parameterName=NormalTexture
   samplerType=SAMPLERTYPE_Normal
   sortPriority=4
   texture=&/T_railgun_impact__N
  }
  n4 = Fmod(A=Time() {period=1}, B=ScalarParameter() {defaultValue=5, group=Time, parameterName=Lifetime})
  n5 = Subtract(A=Time(), B=ScalarParameter() {group=Time, parameterName=Spawntime})
  n6 = CurveAtlasRowParameter(CurveTime=Divide(
    A=StaticSwitchParameter(True=@n4, False=@n5) {group=Time, parameterName=Debug Timestamp, sortPriority=10}
    B=ScalarParameter() {defaultValue=5, group=Time, parameterName=Lifetime}
   )) {
   atlas=&/Curve_Decal_ReactorGrade
   curve=&/C_Decal_Impact_ReactorGrade
   group=Time
   parameterName=TimingCurve
  }
  LinearInterpolate(A=Constant3Vector() {constant=<rgba 0 0 -1 0>}, B=Multiply(A=@n3.RGB, B=Constant3Vector() {constant=<rgba 1 1 -1 0>}), Alpha=@n6.A)
  ScalarParameter() {group=Standard, parameterName=Metallic}
  ScalarParameter() {defaultValue=0.5, group=Standard, parameterName=Specular}
  ScalarParameter() {defaultValue=1, group=Standard, parameterName=Roughness}
  n7 = TextureSampleParameter2D(UVs=@n2) {group=Textures, parameterName=EmissiveTexture01, sortPriority=2, texture=&/T_railgun_impact_B_E2}
  n8 = TextureSampleParameter2D(UVs=@n2) {group=Textures, parameterName=EmissiveTexture02, sortPriority=3, texture=&/T_railgun_impact_A_E}
  n9 = Multiply(
   A=@n6.B
   B=Multiply(
    A=LinearInterpolate(A=@n7.RGB, B=@n8.RGB, Alpha=@n6.R)
    B=ScalarParameter() {defaultValue=5, group=Emissive, parameterName=EmissiveMultiplier}
   ) {constB=5}
  )
  StaticSwitchParameter(
   True=MaterialFunctionCall(LDR Input=Saturate(None=@n9), HDR Tint=@n9) {materialFunction=&DeriveHDRfromLDR}.RemappedRange
   False=Constant()
  ) {defaultValue=true, group=Emissive, parameterName=Use Emissive?, sortPriority=10}
  n10 = ScalarParameter() {defaultValue=1, group=Scorch, parameterName=ScorchTiling}
  n11 = TextureSampleParameter2D(UVs=Add(A=Multiply(A=@n1, B=@n10)) {constB=0.5}) {
   group=Textures
   parameterName=ScorchTexture
   samplerType=SAMPLERTYPE_LinearGrayscale
   sortPriority=1
   texture=&/T_railgun_impact_scorch
  }
  n12 = ScalarParameter() {defaultValue=0.9, group=Standard, parameterName=Opacity}
  Subtract(A=Saturate(None=Multiply(A=@n11.R, B=@n12)), B=OneMinus(None=Saturate(None=@n6.G)))
  VectorParameter() {group=Scorch, parameterName=ScorchColor, defaultValue=<rgba 0.0885 0.0885 0.0885 1>}
 ]
}
