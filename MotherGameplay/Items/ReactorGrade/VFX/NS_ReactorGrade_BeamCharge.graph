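# Niagara charge-up effect for the ReactorGrade item's beam, timed by the WarmupTime and MinWarmupTime user parameters.
# Two Beam mesh particles aim at EndPoint, one stretched by system BeamLength, with LocalPlayer driving an extra mesh visibility tag.
# Flickering, lit Rings spread along X as charge builds, while ChargeGlow, ElectricArcs and Vents start at the minimum-warmup threshold.
# Curves animate scale, color and dynamic material parameters; the Holograms emitter is disabled.
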NiagaraSystem /MotherGameplay/Items/ReactorGrade/VFX/NS_ReactorGrade_BeamCharge
dependencies {
 usedDataInterfaces [&NiagaraDataInterfaceVectorCurve, &NiagaraDataInterfaceCurve, &NiagaraDataInterfaceColorCurve]
 usedDynamicInputs [
  &Multiply_Vector2
  &RandomRangeVector2D
  &Vector2DFromFloat
  &VectorLength
 ]
 usedModules [
  &SystemState
  &self:SystemSpawnScript.SystemScriptSource.SystemScriptGraph.NiagaraNodeAssignment_0.SetVariables_#1
 ]
}
emitters {
 Beam {
  emitterUpdateScript [
   &EmitterState(
    Inactive Response=::NewEnumerator1
    Life Cycle Mode=::NewEnumerator1
    Loop Behavior=::NewEnumerator1
    Loop Duration=$User.WarmupTime
    UseLoopDelay=false
   )
   &SpawnBurst_Instantaneous(Spawn Count=1, Use Spawn Probability=false)
   &SpawnBurst_Instantaneous(Spawn Count=1, Spawn Group=1, Use Spawn Probability=false) {module=SpawnBurst_Instantaneous001}
  ]
  particleSpawnScript [
   /SetVariables(Particles.VisibilityTag=&SpawnGroup()) {isSetParameters=true}
   /SetVariables(
    Particles.LocalMeshPos:Vector3f=<70 0 0>
    Particles.LocalMeshVis:NiagaraInt32=&MakeCustomIntFromBool(Bool=$User.LocalPlayer, True INT=&SpawnGroup(), False INT=2)
   ) {isSetParameters=true}
   &InitializeParticle(
    Color=<rgba 2 0.1 0.1 1>
    Lifetime=$User.WarmupTime
    Mesh Scale=<1 0.35 0.35>
    Mesh Scale Mode=::NewEnumerator4
    UsePositionOffset=false
    Write Material Random=true
   )
   /SetVariables(Particles.Color2=<rgba 0.8 0.01 0.01 1>) {isSetParameters=true}
  ]
  particleUpdateScript [
   &ParticleState()
   &ScaleMeshSize(Scale Factor=&Lerp_Vector(
     A=&VectorFromCurve(
      VectorCurve=DataInterface() {propertyValues {
        XCurve {keys [<at 0 1>]}
        YCurve {keys [
          {arriveTangent=5.7728, interpMode=RCIM_Cubic, leaveTangent=5.7728, tangentMode=RCTM_User, time=0, value=1}
          {arriveTangent=-1.7576, interpMode=RCIM_Cubic, leaveTangent=-1.7576, tangentMode=RCTM_User, time=0.15, value=2}
          {arriveTangent=-1.9955, interpMode=RCIM_Cubic, leaveTangent=-1.9955, tangentMode=RCTM_User, time=1, value=0.7}
         ]}
        ZCurve {keys [
          {arriveTangent=5.7728, interpMode=RCIM_Cubic, leaveTangent=5.7728, tangentMode=RCTM_User, time=0, value=1}
          {arriveTangent=-1.7576, interpMode=RCIM_Cubic, leaveTangent=-1.7576, tangentMode=RCTM_User, time=0.15, value=2}
          {arriveTangent=-1.9955, interpMode=RCIM_Cubic, leaveTangent=-1.9955, tangentMode=RCTM_User, time=1, value=0.7}
         ]}
       }}
      Scale Curve=&MakeVector(X=$System.BeamLength, Y=1, Z=1)
     )
     B=&VectorFromCurve(
      VectorCurve=DataInterface() {propertyValues {
        XCurve {keys [{arriveTangent=4.0081, interpMode=RCIM_Cubic, leaveTangent=4.0081, tangentMode=RCTM_User, time=0, value=0.5}, <at 0.2 1>]}
        YCurve {keys [
          {arriveTangent=3.6119, interpMode=RCIM_Cubic, leaveTangent=3.6119, tangentMode=RCTM_User, time=0, value=0.5}
          {arriveTangentWeight=0.1001, interpMode=RCIM_Cubic, leaveTangentWeight=0.2337, tangentMode=RCTM_User, time=0.3, value=1.25}
          {arriveTangent=-1.5801, interpMode=RCIM_Cubic, leaveTangent=-1.5801, tangentMode=RCTM_User, time=1, value=0.6}
         ]}
        ZCurve {keys [
          {arriveTangent=3.6119, interpMode=RCIM_Cubic, leaveTangent=3.6119, tangentMode=RCTM_User, time=0, value=0.5}
          {arriveTangentWeight=0.1001, interpMode=RCIM_Cubic, leaveTangentWeight=0.2337, tangentMode=RCTM_User, time=0.3, value=1.25}
          {arriveTangent=-1.5801, interpMode=RCIM_Cubic, leaveTangent=-1.5801, tangentMode=RCTM_User, time=1, value=0.6}
         ]}
       }}
      Scale Curve=<4 1.2 1.2>
     )
     Alpha=&MakeFloatFromInt(Int=&SpawnGroup())
    ))
   &DynamicMaterialParameters(
    Index 0 Param 1=&MaskFloatBySpawnGroup(
     Masked Float=0
     Passthrough Float=&FloatFromCurve(FloatCurve=DataInterface() {propertyValues {Curve {keys [
          {arriveTangent=29.814, interpMode=RCIM_Cubic, leaveTangent=29.814, tangentMode=RCTM_User, time=0, value=0}
          {arriveTangent=-2.5249, interpMode=RCIM_Cubic, leaveTangent=-2.5249, tangentMode=RCTM_User, time=0.2, value=2}
          {arriveTangent=-0.0405, interpMode=RCIM_Cubic, leaveTangent=-0.0405, tangentMode=RCTM_User, time=1, value=0.5}
         ]}}})
    )
    Index 0 Param 2=&FloatFromCurve(
     FloatCurve=DataInterface() {propertyValues {Curve {keys [
         {interpMode=RCIM_Cubic, tangentMode=RCTM_User, time=0, value=0}
         {arriveTangent=1.0372, interpMode=RCIM_Cubic, leaveTangent=1.0372, tangentMode=RCTM_User, time=0.5503, value=0.2494}
         {arriveTangent=1.6511, interpMode=RCIM_Cubic, leaveTangent=1.6511, tangentMode=RCTM_User, time=1, value=1}
        ]}}}
     Scale Curve=10
    )
    Index 0 Param 3=&FloatFromCurve(FloatCurve=DataInterface() {propertyValues {Curve {keys [
         {interpMode=RCIM_Cubic, tangentMode=RCTM_User, time=0, value=1}
         {arriveTangent=-1.9967, interpMode=RCIM_Cubic, leaveTangent=-1.9967, tangentMode=RCTM_User, time=1, value=0}
        ]}}})
    Index 0 Param 4=&FloatFromCurve(
     FloatCurve=DataInterface() {propertyValues {Curve {keys [{interpMode=RCIM_Cubic, tangentMode=RCTM_User, time=0, value=1}, {interpMode=RCIM_Cubic, tangentMode=RCTM_User, time=1, value=0.25}]}}}
     Scale Curve=&MaskFloatBySpawnGroup(Masked Float=$System.BeamLength, Passthrough Float=4)
    )
    Module.Param0WriteEnabled=true
    Module.Param1WriteEnabled=true
    Module.Param2WriteEnabled=true
    Module.Param3WriteEnabled=true
   )
   /SetVariables(Particles.DynParams2:Vector4f=&MakeVector4(
     X=&FloatFromCurve(FloatCurve=DataInterface() {propertyValues {Curve {keys [
          {arriveTangent=2.2303, interpMode=RCIM_Cubic, leaveTangent=2.2303, tangentMode=RCTM_User, time=0, value=0}
          {arriveTangent=-0.0405, interpMode=RCIM_Cubic, leaveTangent=-0.0405, tangentMode=RCTM_User, time=1, value=1}
         ]}}})
     Y=0.1
     Z=0
     W=0
    )) {isSetParameters=true}
   /Niagara/Modules/Update/Orientation/V2/OrientMeshToVector(
    Look At Direction:Vector3f=&TransformVector(
     Vector=&Subtract_Vector(A=$User.EndPoint, B=Linked() {linkedVariable {name=Engine.Owner.Position, type=NiagaraPosition}})
     Source Space=::World
     Destination Space=::Simulation
    )
    Override Quat=false
    UseOptionalSideAxis=false
    UseOptionalUpAxis=false
   )
   &ScaleColor(
    Scale Alpha=&FloatFromCurve(FloatCurve=DataInterface() {propertyValues {Curve {keys [<at 0 0 RCIM_Cubic>, {tangentMode=RCTM_User, time=0.25, value=1}]}}})
    Scale RGB=&VectorFromFloat(Value=&FloatFromCurve(FloatCurve=DataInterface() {propertyValues {Curve {keys [<at 0 1 RCIM_Cubic>,<at 1 5>]}}}))
    ScaleA=true
    ScaleRGB=true
   )
  ]
  rendererClasses [&NiagaraMeshRendererProperties]
  renderers [
   {rendererClass=&NiagaraMeshRendererProperties, rendererIndex=0}
   {rendererClass=&NiagaraMeshRendererProperties, rendererIndex=1}
   {rendererClass=&NiagaraMeshRendererProperties, rendererIndex=2}
  ]
 }
 ChargeGlow {
  emitterUpdateScript [
   &EmitterState(
    Enable Distance Culling=true
    Inactive Response=::NewEnumerator1
    Life Cycle Mode=::NewEnumerator1
    Loop Behavior=::NewEnumerator1
    Loop Duration=$User.WarmupTime
    MaxDistance=4000
    Module.bUseMaxDistance=true
    Module.bUseMinDistance=false
    Scalability Mode=::NewEnumerator1
    UseLoopDelay=false
   )
   &SpawnBurst_Instantaneous(Spawn Count=1, Spawn Time=$User.MinWarmupTime, Use Spawn Probability=false)
  ]
  particleSpawnScript [
   &InitializeParticle(
    Color=<rgba 10 0.8683 1.0205 1>
    Lifetime=&Subtract_Float(A=$User.WarmupTime, B=$User.MinWarmupTime)
    Mass=0
    Mass Mode=::NewEnumerator1
    Mesh Scale=<0.1 0.08 0.1>
    Mesh Scale Mode=::NewEnumerator4
    Position Offset=<-26 0 2>
    Sprite Size Mode=::NewEnumerator3
    Uniform Sprite Size=17
    UsePositionOffset=true
    Write Material Random=true
   )
   &CameraOffset(Camera Offset Amount=3)
   /SetVariables(Particles.MeshPosition:Vector3f=<-31.1 0 10.7>) {isSetParameters=true, module=SetVariables_#2}
   &InitialMeshOrientation(Orientation Vector=<1 0 -0.7>, Use Orientation Vector=true, Use Rotation Vector=false)
   /SetVariables(
    Particles.TubeDynParams:Vector4f=<5 10 0.5 0>
    Particles.TubeSpriteFacing:Vector3f=<1 0 0>
    Particles.TubeSpritePosition:Vector3f=<28 0 -3>
    Particles.TubeSpriteSize:Vector2f=<6 55>
   ) {isSetParameters=true}
  ]
  particleUpdateScript [
   &ParticleState()
   &DynamicMaterialParameters(
    Index 0 Param 1=&FloatFromCurve(FloatCurve=DataInterface() {propertyValues {Curve {keys [
         {arriveTangent=-16.9663, interpMode=RCIM_Cubic, leaveTangent=-16.9663, tangentMode=RCTM_User, time=0, value=1}
         {arriveTangent=2.601, interpMode=RCIM_Cubic, leaveTangent=2.601, tangentMode=RCTM_User, time=0.15, value=0.25}
         {arriveTangent=-1.1077, interpMode=RCIM_Cubic, leaveTangent=-1.1077, tangentMode=RCTM_User, time=1, value=0}
        ]}}})
    Index 0 Param 2=1
    Index 1 Param 1=12
    Module.Param0WriteEnabled=true
    Module.Param0WriteEnabled1=true
    Module.Param1WriteEnabled=true
    Module.Param1WriteEnabled1=false
    Module.Param2WriteEnabled=false
    Module.Param2WriteEnabled1=false
    Module.Param3WriteEnabled=false
    Module.Param3WriteEnabled1=false
    Write Parameter Index 1=true
   )
   &ScaleColor(
    Scale Alpha=&FloatFromCurve(FloatCurve=DataInterface() {propertyValues {Curve {keys [
         {arriveTangent=143.6838, interpMode=RCIM_Cubic, leaveTangent=143.6838, tangentMode=RCTM_User, time=0, value=1}
         {arriveTangent=-26.0764, interpMode=RCIM_Cubic, leaveTangent=-26.0764, tangentMode=RCTM_User, time=0.05, value=4}
         <at 0.3 1>
         {arriveTangentWeight=0.1163, interpMode=RCIM_Cubic, leaveTangentWeight=0.2906, tangentMode=RCTM_User, time=0.5, value=1}
         {arriveTangent=8.8017, interpMode=RCIM_Cubic, leaveTangent=8.8017, tangentMode=RCTM_User, time=1, value=3}
        ]}}})
    ScaleA=true
    ScaleRGB=false
   )
   /SetVariables(Particles.MeshColor:LinearColor=&MultiplyLinearColor(
     Linear Color=&ColorFromCurve(ColorCurve=DataInterface() {propertyValues {
        AlphaCurve {keys [
          {interpMode=RCIM_Cubic, leaveTangentWeight=0.3333, tangentMode=RCTM_User, time=0, value=1}
          {arriveTangentWeight=0.3333, interpMode=RCIM_Cubic, tangentMode=RCTM_User, time=1, value=1}
         ]}
        BlueCurve {keys [
          <at 0.0733 0.5625>
          {arriveTangentWeight=0.1192, interpMode=RCIM_Cubic, leaveTangentWeight=1.564, tangentMode=RCTM_User, time=0.1, value=0.6615}
          {arriveTangentWeight=1.4289, interpMode=RCIM_Cubic, leaveTangentWeight=0.0921, tangentMode=RCTM_User, time=0.782, value=0.4611}
          {arriveTangentWeight=0.0141, interpMode=RCIM_Cubic, tangentMode=RCTM_User, time=0.8133, value=0.3532}
         ]}
        GreenCurve {keys [
          <at 0.0733 0.6787>
          {arriveTangentWeight=0.0863, interpMode=RCIM_Cubic, leaveTangentWeight=1.1321, tangentMode=RCTM_User, time=0.1, value=0.5371}
          {arriveTangentWeight=0.777, interpMode=RCIM_Cubic, leaveTangentWeight=0.0501, tangentMode=RCTM_User, time=0.782, value=0.3333}
          {arriveTangentWeight=0.0141, interpMode=RCIM_Cubic, tangentMode=RCTM_User, time=0.8133, value=0.5223}
         ]}
        RedCurve {keys [
          <at 0.0733 12>
          {arriveTangentWeight=0.1164, interpMode=RCIM_Cubic, leaveTangentWeight=1.528, tangentMode=RCTM_User, time=0.1, value=8}
          {arriveTangentWeight=2.4099, interpMode=RCIM_Cubic, leaveTangentWeight=0.1554, tangentMode=RCTM_User, time=0.782, value=8}
          {arriveTangentWeight=0.0141, interpMode=RCIM_Cubic, tangentMode=RCTM_User, time=0.8133, value=17}
         ]}
       }})
     Scale Factor=&MakeVector4_FromFloat(Float=&Waveform(
       Global Phase=$Particles.Age
       Global Clamp Min=&FloatFromCurve(FloatCurve=DataInterface() {propertyValues {Curve {keys [
            {arriveTangent=24.8524, interpMode=RCIM_Constant, time=0, value=0.25}
            {arriveTangentWeight=0.0098, interpMode=RCIM_Cubic, leaveTangentWeight=0.3444, tangentMode=RCTM_User, time=0.25, value=1}
            <at 0.75 1>
            {arriveTangent=-5.6413, interpMode=RCIM_Cubic, leaveTangent=-5.6413, tangentMode=RCTM_User, time=0.755, value=0}
           ]}}})
       "Waveform [1]"=::NewEnumerator0
       "[1] Phase"=$Particles.Age
       "[1] Frequency"=16
       "[1] Amplitude Min/Max"=<-4 4>
       "[2] Phase"=$Emitter.Age
       "[3] Phase"=$Emitter.Age
      ))
    )) {isSetParameters=true, module=SetVariables_#3}
   &ScaleMeshSize(Scale Factor=&VectorFromCurve(VectorCurve=DataInterface() {propertyValues {
       XCurve {keys [<at 0 1>,<at 1 1>]}
       YCurve {keys [<at 0 0.3>, {arriveTangent=7.2375, interpMode=RCIM_Cubic, leaveTangent=7.2375, tangentMode=RCTM_User, time=0.1, value=1}, <at 0.2 1>]}
       ZCurve {keys [<at 0 1.5>, {arriveTangent=-5.3018, interpMode=RCIM_Cubic, leaveTangent=-5.3018, tangentMode=RCTM_User, time=0.1, value=1}, <at 0.2 1>]}
      }}))
  ]
  rendererClasses [
   &NiagaraSpriteRendererProperties
   &NiagaraLightRendererProperties
   &NiagaraMeshRendererProperties
  ]
  renderers [
   {rendererClass=&NiagaraSpriteRendererProperties}
   {rendererClass=&NiagaraLightRendererProperties, rendererIndex=1}
   {rendererClass=&NiagaraMeshRendererProperties, rendererIndex=2}
   {rendererClass=&NiagaraSpriteRendererProperties, rendererIndex=3}
  ]
 }
 ElectricArcs {
  emitterUpdateScript [
   &EmitterState(
    Enable Distance Culling=true
    Life Cycle Mode=::NewEnumerator1
    Loop Behavior=::NewEnumerator1
    Loop Delay=$User.MinWarmupTime
    Loop Duration=&Subtract_Float(A=$User.WarmupTime, B=$User.MinWarmupTime)
    MaxDistance=2000
    Module.bUseMaxDistance=true
    Module.bUseMinDistance=false
    Scalability Mode=::NewEnumerator1
    UseLoopDelay=true
   )
   &SpawnRate(
    SpawnRate=&FloatFromCurve(
     FloatCurve=DataInterface() {propertyValues {Curve {keys [<at 0 0 RCIM_Cubic>,<at 0.8 1>,<at 0.81 0>]}}}
     CurveIndex=$System.NormalizedLoopAge
     Scale Curve=17
    )
    Use Spawn Probability=true
   )
   &SpawnBurst_Instantaneous(Spawn Count=2, Use Spawn Probability=false)
  ]
  particleSpawnScript [
   &InitializeParticle(
    Color=<rgba 2 0.1336 0.1647 1>
    Lifetime=&FloatFromCurve(
     FloatCurve=DataInterface() {propertyValues {Curve {keys [<at 0 1 RCIM_Cubic>,<at 0.9 0.5>]}}}
     CurveIndex=$System.NormalizedLoopAge
     Scale Curve=&UniformRangedFloat(Minimum=0.4, Maximum=0.7)
    )
    Mesh Scale Max=<1.5 1 2>
    Mesh Scale Min=<0.7 0.5 1>
    Mesh Scale Mode=::NewEnumerator2
    Sprite Rotation Mode=::NewEnumerator2
    UsePositionOffset=false
    Write Material Random=true
   ) {module=InitializeParticle001}
   &InitialMeshOrientation(
    Orientation Vector=<-1 0 0>
    Rotation=&MakeVector(
     X=&FloatFromCurve(
      FloatCurve=DataInterface() {propertyValues {Curve {keys [
          {arriveTangent=-5.3457, interpMode=RCIM_Cubic, time=0, value=1}
          {
           arriveTangent=-2.5696
           arriveTangentWeight=0.3542
           interpMode=RCIM_Cubic
           leaveTangent=-2.5696
           leaveTangentWeight=0.3542
           tangentMode=RCTM_User
           time=0.5
           value=0.5
          }
          {arriveTangent=-2.7059, interpMode=RCIM_Cubic, leaveTangent=-2.7059, tangentMode=RCTM_User, time=0.5, value=-0.5}
          {interpMode=RCIM_Cubic, leaveTangent=-3.5233, time=1, value=-1}
         ]}}}
      CurveIndex=$Particles.MaterialRandom
      Scale Curve=0.25
     )
     Y=0
     Z=0
    )
    Use Orientation Vector=true
    Use Rotation Vector=true
   ) {module=InitialMeshOrientation001}
   &DynamicMaterialParameters(
    Index 0 Param 3=&RandomRangeFloat(Minimum=0, Maximum=1)
    Index 0 Param 4=&UniformRangedFloat(Minimum=1.5, Maximum=2.5)
    Module.Param0WriteEnabled=false
    Module.Param1WriteEnabled=false
    Module.Param2WriteEnabled=true
    Module.Param3WriteEnabled=true
   ) {module=DynamicMaterialParameters001}
   /SetVariables(Particles.Position=&Lerp_Vector(A=<10 0 -4>, B=<0 0 -4>, Alpha=&RandomRangeFloat(Minimum=0, Maximum=1))) {isSetParameters=true}
   &CameraOffset(Camera Offset Amount=3)
  ]
  particleUpdateScript [
   &ParticleState()
   /SetVariables(Particles.Position=&MakeVector(
     X=&Lerp_Float(
      A=&MakeFloatFromVector(Vector=Linked() {linkedVariable {name=Particles.Initial.Position, type=NiagaraPosition}}, Channel=::NewEnumerator0)
      B=60
      Alpha=&FloatFromCurve(FloatCurve=DataInterface() {propertyValues {Curve {keys [<at 0 0 RCIM_Cubic>, {arriveTangent=1.5951, interpMode=RCIM_Cubic, leaveTangent=1.5951, tangentMode=RCTM_User, time=1, value=1}]}}})
     )
     Y=&MakeFloatFromVector(Vector=Linked() {linkedVariable {name=Particles.Initial.Position, type=NiagaraPosition}}, Channel=::NewEnumerator1)
     Z=-4
    )) {isSetParameters=true}
   &ScaleColor(
    Scale Alpha=&self:Flicker_01(
     FlickerAmount:NiagaraFloat=&FloatFromCurve(
      FloatCurve=DataInterface() {propertyValues {Curve {keys [
          {arriveTangent=9.8056, interpMode=RCIM_Cubic, time=0, value=1}
          {interpMode=RCIM_Cubic, tangentMode=RCTM_User, time=0.331, value=0.2611}
          {interpMode=RCIM_Cubic, tangentMode=RCTM_User, time=0.7, value=0.45}
          {interpMode=RCIM_Cubic, leaveTangent=-11960.9277, time=1, value=0}
         ]}}}
      Scale Curve=2
     )
     FlickerInterval:NiagaraFloat=0.1
    )
    ScaleA=true
    ScaleRGB=true
   )
   &DynamicMaterialParameters(
    Index 0 Param 1=&FloatFromCurve(
     FloatCurve=DataInterface() {propertyValues {Curve {keys [
         {arriveTangent=12.9112, interpMode=RCIM_Cubic, leaveTangent=12.9112, tangentMode=RCTM_User, time=0, value=0.25}
         <at 0.1 1>
         {arriveTangent=-1.1111, interpMode=RCIM_Cubic, leaveTangent=-1.1111, time=0.5, value=1}
         {arriveTangent=-2.5142, interpMode=RCIM_Cubic, leaveTangent=-2.5142, tangentMode=RCTM_User, time=1, value=0}
        ]}}}
     CurveIndex=$Particles.Age
     Scale Curve=&Lerp_Float(A=0.4, B=0.7, Alpha=$Particles.Initial.MaterialRandom)
    )
    Index 0 Param 2=&FloatFromCurve(
     FloatCurve=DataInterface() {propertyValues {Curve {keys [
         {arriveTangent=10.9494, interpMode=RCIM_Cubic, leaveTangent=10.9494, tangentMode=RCTM_User, time=0, value=1}
         {arriveTangent=-8.5345, interpMode=RCIM_Cubic, leaveTangent=-8.5345, tangentMode=RCTM_User, time=0.2, value=2}
         {arriveTangent=-2.4364, interpMode=RCIM_Cubic, leaveTangent=-2.4364, time=0.3638, value=0.5858}
         {arriveTangent=-8.496, interpMode=RCIM_Cubic, leaveTangent=-8.496, tangentMode=RCTM_User, time=0.4639, value=1.357}
         {arriveTangent=8.1076, interpMode=RCIM_Cubic, leaveTangent=-1.1376, tangentMode=RCTM_User, time=1, value=1.5}
        ]}}}
     CurveIndex=$Particles.Age
     Scale Curve=0.25
    )
    Index 0 Param 4=&RandomRangeFloat(Minimum=0.6, Maximum=0.8)
    Module.Param0WriteEnabled=true
    Module.Param1WriteEnabled=false
    Module.Param2WriteEnabled=false
    Module.Param3WriteEnabled=true
   )
   /SetVariables(Particles.DynamicParamOverride:Vector4f=&MakeFloat4FromVectorAndFloat(
     "Vector (RGB)"=/Niagara/DynamicInputs/Transforms/TransformPosition(Position=Linked() {linkedVariable {name=Particles.Initial.Position, type=NiagaraPosition}}, Source Space=::Local, Destination Space=::World)
     "Float (Alpha)"=0
    )) {isSetParameters=true}
   &ScaleMeshSize(Scale Factor=&VectorFromCurve(
     VectorCurve=DataInterface() {propertyValues {
       XCurve {keys [
         {arriveTangent=2.9222, interpMode=RCIM_Cubic, leaveTangent=2.9222, tangentMode=RCTM_User, time=0, value=0}
         {interpMode=RCIM_Cubic, tangentMode=RCTM_User, time=0.65, value=1}
        ]}
       YCurve {keys [<at 0 1>]}
       ZCurve {keys [
         {arriveTangent=3.2926, interpMode=RCIM_Cubic, leaveTangent=3.2925, tangentMode=RCTM_User, time=0, value=0.1}
         {arriveTangent=-1.316, interpMode=RCIM_Cubic, leaveTangent=-1.316, tangentMode=RCTM_User, time=0.2588, value=1}
         {arriveTangent=0.348, interpMode=RCIM_Cubic, leaveTangent=0.348, tangentMode=RCTM_User, time=0.4597, value=0.4069}
         {interpMode=RCIM_Cubic, leaveTangent=3.5341, time=1, value=1.5}
        ]}
      }}
     CurveIndex=$Particles.Age
    ))
  ]
  rendererClasses [&NiagaraMeshRendererProperties]
  renderers [{rendererClass=&NiagaraMeshRendererProperties, rendererIndex=0}]
 }
 Holograms {
  bEnabled=false
  emitterSpawnScript [/SetVariables(Emitter.Timing:NiagaraFloat=&Divide_Float(A=&Subtract_Float(A=$User.WarmupTime, B=$User.MinWarmupTime), B=4)) {isSetParameters=true}]
  emitterUpdateScript [
   &EmitterState(
    Inactive Response=::NewEnumerator1
    Life Cycle Mode=::NewEnumerator1
    Loop Behavior=::NewEnumerator1
    Loop Delay=$User.MinWarmupTime
    Loop Duration=&Subtract_Float(A=$User.WarmupTime, B=$User.MinWarmupTime)
    UseLoopDelay=true
   )
   &SpawnBurst_Instantaneous(Spawn Count=1, Spawn Time=$Emitter.Timing, Use Spawn Probability=false)
   &SpawnBurst_Instantaneous(Spawn Count=1, Spawn Time=&Multiply_Float(A=$Emitter.Timing, B=2), Use Spawn Probability=false) {module=SpawnBurst_Instantaneous001}
   &SpawnBurst_Instantaneous(Spawn Count=1, Spawn Time=&Multiply_Float(A=$Emitter.Timing, B=3), Use Spawn Probability=false) {module=SpawnBurst_Instantaneous002}
  ]
  particleSpawnScript [
   &FN_SetSpawnGroup(Number of Spawn Groups=3)
   &InitializeParticle(
    Color=<rgba 0.8683 5.0657 10 1>
    Mass=0
    Mass Mode=::NewEnumerator1
    Mesh Scale=<0.1 0.09 0.11>
    Mesh Scale Mode=::NewEnumerator4
    Position Offset=&MakeVector(
     X=&Lerp_Float(
      A=-26
      B=-21
      Alpha=&Divide_Float(A=&MakeFloatFromInt(Int=&SpawnGroup()), B=3)
     )
     Y=0
     Z=11
    )
    UsePositionOffset=true
    Write Material Random=true
   )
   &CameraOffset(Camera Offset Amount=3)
   /SetVariables(Particles.MeshPosition:Vector3f=<-30.1 0 10.7>) {isSetParameters=true, module=SetVariables_#2}
   &InitialMeshOrientation(Use Orientation Vector=true, Use Rotation Vector=false)
  ]
  particleUpdateScript [
   &ParticleState(Kill Particles When Lifetime Has Elapsed=false)
   &DynamicMaterialParameters(
    Index 0 Param 1=&FloatFromCurve(FloatCurve=DataInterface() {propertyValues {Curve {keys [
         {arriveTangent=-16.9663, interpMode=RCIM_Cubic, leaveTangent=-16.9663, tangentMode=RCTM_User, time=0, value=1}
         {arriveTangent=2.601, interpMode=RCIM_Cubic, leaveTangent=2.601, tangentMode=RCTM_User, time=0.15, value=0.25}
         {arriveTangent=-1.1077, interpMode=RCIM_Cubic, leaveTangent=-1.1077, tangentMode=RCTM_User, time=1, value=0}
        ]}}})
    Index 0 Param 2=1
    Index 1 Param 1=12
    Module.Param0WriteEnabled=true
    Module.Param0WriteEnabled1=true
    Module.Param1WriteEnabled=true
    Module.Param1WriteEnabled1=false
    Module.Param2WriteEnabled=false
    Module.Param2WriteEnabled1=false
    Module.Param3WriteEnabled=false
    Module.Param3WriteEnabled1=false
    Write Parameter Index 1=true
   )
   &ScaleColor(
    Scale Alpha=&FloatFromCurve(FloatCurve=DataInterface() {propertyValues {Curve {keys [
         {arriveTangent=155.9122, interpMode=RCIM_Cubic, leaveTangent=155.9122, tangentMode=RCTM_User, time=0, value=1}
         {arriveTangent=-26.0764, interpMode=RCIM_Cubic, leaveTangent=-26.0764, tangentMode=RCTM_User, time=0.05, value=4}
         <at 0.3 1>
         {arriveTangentWeight=0.1163, interpMode=RCIM_Cubic, leaveTangentWeight=0.2906, tangentMode=RCTM_User, time=0.5, value=1}
         {arriveTangent=8.8017, interpMode=RCIM_Cubic, leaveTangent=8.8017, tangentMode=RCTM_User, time=1, value=3}
        ]}}})
    ScaleA=true
    ScaleRGB=false
   )
   /SetVariables(Particles.MeshColor:LinearColor=&MultiplyLinearColor(
     Linear Color=&ColorFromCurve(
      ColorCurve=DataInterface() {propertyValues {
        AlphaCurve {keys [
          {interpMode=RCIM_Cubic, leaveTangentWeight=0.3333, tangentMode=RCTM_User, time=0, value=1}
          {arriveTangentWeight=0.3333, interpMode=RCIM_Cubic, tangentMode=RCTM_User, time=1, value=1}
         ]}
        BlueCurve {keys [
          {interpMode=RCIM_Cubic, leaveTangentWeight=0.0167, tangentMode=RCTM_User, time=0.05, value=12}
          {arriveTangentWeight=0.1192, interpMode=RCIM_Cubic, leaveTangentWeight=1.564, tangentMode=RCTM_User, time=0.1, value=8}
          {arriveTangentWeight=1.4289, interpMode=RCIM_Cubic, leaveTangentWeight=0.0921, tangentMode=RCTM_User, time=0.782, value=8}
          {arriveTangentWeight=0.0141, interpMode=RCIM_Cubic, tangentMode=RCTM_User, time=0.8133, value=0}
         ]}
        GreenCurve {keys [
          {interpMode=RCIM_Cubic, leaveTangentWeight=0.0167, tangentMode=RCTM_User, time=0.05, value=10.3033}
          {arriveTangentWeight=0.0863, interpMode=RCIM_Cubic, leaveTangentWeight=1.1321, tangentMode=RCTM_User, time=0.1, value=4.3398}
          {arriveTangentWeight=0.777, interpMode=RCIM_Cubic, leaveTangentWeight=0.0501, tangentMode=RCTM_User, time=0.782, value=8}
          {arriveTangentWeight=0.0141, interpMode=RCIM_Cubic, tangentMode=RCTM_User, time=0.8133, value=2.7431}
         ]}
        RedCurve {keys [
          {interpMode=RCIM_Cubic, leaveTangentWeight=0.0167, tangentMode=RCTM_User, time=0.05, value=6.1091}
          {arriveTangentWeight=0.1164, interpMode=RCIM_Cubic, leaveTangentWeight=1.528, tangentMode=RCTM_User, time=0.1, value=0.5371}
          {arriveTangentWeight=2.4099, interpMode=RCIM_Cubic, leaveTangentWeight=0.1554, tangentMode=RCTM_User, time=0.782, value=0.3333}
          {arriveTangentWeight=0.0141, interpMode=RCIM_Cubic, tangentMode=RCTM_User, time=0.8133, value=15}
         ]}
       }}
      CurveIndex=$Emitter.NormalizedLoopAge
     )
     Scale Factor=&MakeVector4_FromFloat(Float=&Waveform(
       Global Phase=$Particles.Age
       Global Clamp Min=&FloatFromCurve(
        FloatCurve=DataInterface() {propertyValues {Curve {keys [
            {arriveTangent=24.8524, interpMode=RCIM_Constant, time=0, value=0.1}
            {arriveTangentWeight=0.0098, interpMode=RCIM_Cubic, leaveTangentWeight=0.3444, tangentMode=RCTM_User, time=0.25, value=1}
            <at 0.6 1>
            {arriveTangent=-5.6413, interpMode=RCIM_Cubic, leaveTangent=-5.6413, tangentMode=RCTM_User, time=1, value=0}
           ]}}}
        CurveIndex=$Emitter.NormalizedLoopAge
       )
       "Waveform [1]"=::NewEnumerator0
       "[1] Phase"=$Particles.Age
       "[1] Frequency"=10
       "[1] Amplitude Min/Max"=<-2 2>
       "[2] Phase"=$Emitter.Age
       "[3] Phase"=$Emitter.Age
      ))
    )) {isSetParameters=true, module=SetVariables_#3}
   &ScaleMeshSize(Scale Factor=&VectorFromCurve(VectorCurve=DataInterface() {propertyValues {
       XCurve {keys [<at 0 1>,<at 1 1>]}
       YCurve {keys [<at 0 0.3>, {arriveTangent=7.2375, interpMode=RCIM_Cubic, leaveTangent=7.2375, tangentMode=RCTM_User, time=0.1, value=1}, <at 0.2 1>]}
       ZCurve {keys [<at 0 1.5>, {arriveTangent=-5.3018, interpMode=RCIM_Cubic, leaveTangent=-5.3018, tangentMode=RCTM_User, time=0.1, value=1}, <at 0.2 1>]}
      }}))
  ]
  rendererClasses [&NiagaraMeshRendererProperties]
  renderers [{rendererClass=&NiagaraMeshRendererProperties, rendererIndex=0}]
 }
 Rings {
  emitterUpdateScript [
   &EmitterState(Enable Distance Culling=true, MaxDistance=3500, Module.bUseMaxDistance=true, Module.bUseMinDistance=false, Scalability Mode=::NewEnumerator1)
   &SpawnRate(
    SpawnRate=&FloatFromCurve(
     FloatCurve=DataInterface() {propertyValues {Curve {keys [<at 0 1 RCIM_Cubic>, {arriveTangent=0.6098, interpMode=RCIM_Cubic, leaveTangent=0.6098, tangentMode=RCTM_User, time=1, value=1.5}]}}}
     CurveIndex=$System.NormalizedLoopAge
     Scale Curve=100
    )
    Use Spawn Probability=false
   )
  ]
  particleSpawnScript [
   &FN_SetSpawnGroup_v2(Number of Spawn Groups=16)
   &InitializeParticle(
    Color=<rgba 25 1.5 0.7 1>
    Lifetime Max=0.16
    Lifetime Min=0.1
    Lifetime Mode=::NewEnumerator1
    Position Offset=&MakeVector(
     X=&RandomRangeFloat(
      Minimum=0
      Maximum=&FloatFromCurve(
       FloatCurve=DataInterface() {propertyValues {Curve {keys [{interpMode=RCIM_Cubic, tangentMode=RCTM_User, time=0, value=0}, {interpMode=RCIM_Cubic, tangentMode=RCTM_User, time=0.7, value=1}]}}}
       CurveIndex=$System.NormalizedLoopAge
       Scale Curve=60
      )
     )
     Y=0
     Z=0
    )
    Sprite Rotation Mode=::NewEnumerator2
    Sprite Size Mode=::NewEnumerator3
    Uniform Sprite Size=&Multiply_Float(
     A=&RandomRangeFloat(Minimum=14, Maximum=20)
     B=&MakeCustomFloatFromBool(Bool=&SetBoolByFloatComparison(A=$System.Age, B=$User.MinWarmupTime), True Float=1, False Float=0.6)
    )
    UsePositionOffset=true
    Write Material Random=true
   ) {module=InitializeParticle001}
   &SubUVAnimation(End Frame=3, SubUV Animation Mode:ENiagara_SubUVLookupMode=::NewEnumerator2) {module=SubUVAnimation001}
   &DynamicMaterialParameters(
    Index 0 Param 1=$Particles.MaterialRandom
    Index 0 Param 2=&Lerp_Float(A=0.5, B=1, Alpha=$Particles.MaterialRandom)
    Module.Param0WriteEnabled=true
    Module.Param1WriteEnabled=true
    Module.Param2WriteEnabled=false
    Module.Param3WriteEnabled=false
   )
   /SetVariables(Particles.LightExp=6) {isSetParameters=true}
   /SetVariables(Particles.LightEnabled:NiagaraBool=&SetBoolByIntComparison(A=0, B=&SpawnGroup(), Comparison Type=::NewEnumerator4)) {isSetParameters=true}
   /SetVariables(Particles.SpritePos2:Vector3f=&Add_Vector(A=Linked() {linkedVariable {name=Particles.Position, type=NiagaraPosition}}, B=<0.5 0.25 0.25>)) {isSetParameters=true}
   &CameraOffset(Camera Offset Amount=5)
   /NiagaraFN/Sprite/FN_SpriteAlignmentAndFacing(SpriteAlignment:Vector3f=$Engine.Owner.SystemXAxis, SpriteFacing:Vector3f=$Engine.Owner.SystemXAxis)
  ]
  particleUpdateScript [
   &ParticleState()
   &SolveForcesAndVelocity(Clamp Velocity=false, Limit Acceleration=false)
   &DynamicMaterialParameters(
    Index 0 Param 3=&FloatFromCurve(FloatCurve=DataInterface() {propertyValues {Curve {keys [<at 0.2 0 RCIM_Cubic>, {interpMode=RCIM_Cubic, leaveTangent=1.5819, time=1, value=1}]}}})
    Module.Param0WriteEnabled=false
    Module.Param1WriteEnabled=false
    Module.Param2WriteEnabled=true
    Module.Param3WriteEnabled=false
   ) {module=DynamicMaterialParameters001}
   /SetVariables(Particles.LightColor=&MultiplyLinearColor(
     Linear Color=$Particles.Initial.Color
     Scale Factor=&MakeVector4_FromFloat(Float=&Multiply_Float(
       A=10
       B=&self:Flicker(
        FlickerAmount:NiagaraFloat=&FloatFromCurve(
         FloatCurve=DataInterface() {propertyValues {Curve {keys [
             {arriveTangent=9.8056, interpMode=RCIM_Cubic, time=0, value=1}
             {interpMode=RCIM_Cubic, tangentMode=RCTM_User, time=0.331, value=0.2611}
             {interpMode=RCIM_Cubic, tangentMode=RCTM_User, time=0.6768, value=0.8316}
             {arriveTangent=-3.2376, interpMode=RCIM_Cubic, leaveTangent=-11960.9277, tangentMode=RCTM_User, time=1, value=0.1}
            ]}}}
         CurveIndex=$Particles.Age
        )
        FlickerInterval:NiagaraFloat=0.1
       )
      ))
    )) {isSetParameters=true}
   &SpriteRotationRate(Rotation Rate=&RandomRangeFloat(Minimum=-250, Maximum=250))
   &ScaleColor(
    Scale Alpha=&FloatFromCurve(FloatCurve=DataInterface() {propertyValues {Curve {keys [
         {arriveTangent=9.0725, interpMode=RCIM_Cubic, leaveTangent=9.0725, tangentMode=RCTM_User, time=0, value=0}
         <at 0.25 1 RCIM_Cubic>
         {arriveTangent=-0.0405, interpMode=RCIM_Cubic, leaveTangent=-0.0405, tangentMode=RCTM_User, time=1, value=0}
        ]}}})
    ScaleA=true
    ScaleRGB=false
   )
   /SetVariables(Particles.SpriteColor2:LinearColor=<rgba 4 0.6695 3.083 0.5>) {isSetParameters=true}
  ]
  rendererClasses [&NiagaraSpriteRendererProperties, &NiagaraLightRendererProperties]
  renderers [
   {rendererClass=&NiagaraSpriteRendererProperties}
   {rendererClass=&NiagaraLightRendererProperties, rendererIndex=1}
   {rendererClass=&NiagaraSpriteRendererProperties, rendererIndex=2}
  ]
 }
 Vents {
  emitterUpdateScript [
   &EmitterState(Enable Distance Culling=true, MaxDistance=2000, Module.bUseMaxDistance=true, Module.bUseMinDistance=false, Scalability Mode=::NewEnumerator1)
   &SpawnBurst_Instantaneous(Spawn Count=1, Spawn Time=$User.MinWarmupTime, Use Spawn Probability=false)
  ]
  particleSpawnScript [
   &InitializeParticle(
    Color=<rgba 2 0.1875 0.2177 1>
    Mass=0
    Mass Mode=::NewEnumerator1
    Mesh Scale Mode=::NewEnumerator3
    Position Offset=<-9.9 0 -0.25>
    Sprite Size Mode=::NewEnumerator3
    Uniform Sprite Size=17
    UsePositionOffset=true
    Write Material Random=true
   )
   &InitialMeshOrientation(Use Orientation Vector=true, Use Rotation Vector=false)
  ]
  particleUpdateScript [
   &ParticleState()
   &ScaleColor(
    Scale Alpha=&FloatFromCurve(FloatCurve=DataInterface() {propertyValues {Curve {keys [{arriveTangent=40.829, interpMode=RCIM_Cubic, leaveTangent=40.829, tangentMode=RCTM_User, time=0, value=0}, <at 0.05 1 RCIM_Cubic>, <at 1 0>]}}})
    ScaleA=true
    ScaleRGB=false
   )
   &DynamicMaterialParameters(
    Index 0 Param 1=&FloatFromCurve(FloatCurve=DataInterface() {propertyValues {Curve {keys [
         {arriveTangent=1.9352, interpMode=RCIM_Cubic, leaveTangent=1.9352, tangentMode=RCTM_User, time=0, value=0.1}
         {arriveTangent=0.3565, interpMode=RCIM_Cubic, leaveTangent=0.3565, tangentMode=RCTM_User, time=1, value=1}
        ]}}})
    Index 0 Param 2=2
    Index 1 Param 2=&FloatFromCurve(
     FloatCurve=DataInterface() {propertyValues {Curve {keys [{interpMode=RCIM_Cubic, tangentMode=RCTM_User, time=0, value=1}, {interpMode=RCIM_Cubic, tangentMode=RCTM_User, time=1, value=0}]}}}
     Scale Curve=5
    )
    Module.Param0WriteEnabled=true
    Module.Param0WriteEnabled1=true
    Module.Param1WriteEnabled=true
    Module.Param1WriteEnabled1=true
    Module.Param2WriteEnabled=false
    Module.Param2WriteEnabled1=false
    Module.Param3WriteEnabled=false
    Module.Param3WriteEnabled1=false
    Write Parameter Index 1=true
   ) {module=DynamicMaterialParameters001}
   &ScaleMeshSize(Scale Factor=&MakeVector(
     X=1
     Y=&FloatFromCurve(FloatCurve=DataInterface() {propertyValues {Curve {keys [
          {arriveTangent=11.2887, interpMode=RCIM_Cubic, leaveTangent=11.2887, tangentMode=RCTM_User, time=0, value=0}
          {arriveTangent=-0.0405, interpMode=RCIM_Cubic, leaveTangent=-0.0405, tangentMode=RCTM_User, time=0.25, value=1}
         ]}}})
     Z=1
    ))
  ]
  rendererClasses [&NiagaraMeshRendererProperties]
  renderers [{rendererClass=&NiagaraMeshRendererProperties, rendererIndex=0}]
 }
}
summary {
 systemName=NS_ReactorGrade_BeamCharge
 emitters [
  {emitterName=Beam, rendererClasses [&NiagaraMeshRendererProperties]}
  {emitterName=Rings, rendererClasses [&NiagaraSpriteRendererProperties, &NiagaraLightRendererProperties]}
  {emitterName=ElectricArcs, rendererClasses [&NiagaraMeshRendererProperties]}
  {
   emitterName=ChargeGlow
   rendererClasses [
    &NiagaraSpriteRendererProperties
    &NiagaraLightRendererProperties
    &NiagaraMeshRendererProperties
   ]
  }
  {emitterName=Vents, rendererClasses [&NiagaraMeshRendererProperties]}
  {bEnabled=false, emitterName=Holograms, rendererClasses [&NiagaraMeshRendererProperties]}
 ]
 userVariables [
  User.EndPoint {type=Vector3f, description="", defaultValue {struct=Vector3f, value=<3000 0 2000>}}
  User.LocalPlayer {type=NiagaraBool}
  User.MinWarmupTime {type=NiagaraFloat, defaultValue=1.25, description=""}
  User.WarmupTime {type=NiagaraFloat, defaultValue=2.5, description=""}
 ]
}
