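# Niagara effect for the ReactorGrade item's beam shot, a red-pink beam from the system origin toward User.BeamEnd_0.
# A single mesh is stretched by System.BeamTile along FireDirection and fades over User.Duration.
# GPU sparks and delayed residual sprites spawn along the line in tile-scaled counts, drift with noise, and die beyond beam length.
# A tracer with mesh, sprite and light renderers uses distance-scaled, frame-limited scratch-pad timing to travel to the endpoint.
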NiagaraSystem /MotherGameplay/Items/ReactorGrade/VFX/NS_ReactorGrade_BeamFire
dependencies {
 usedDataInterfaces [&NiagaraDataInterfaceCurve, &NiagaraDataInterfaceColorCurve, &NiagaraDataInterfaceVectorCurve]
 usedDynamicInputs [&NormalizeVector]
 usedModules [
  &self:SystemSpawnScript.SystemScriptSource.SystemScriptGraph.NiagaraNodeAssignment_0.SetVariables_#1
  &SystemState
 ]
}
emitters {
 Beam {
  emitterUpdateScript [
   &EmitterState()
   &SpawnBurst_Instantaneous(Spawn Count=1, Use Spawn Probability=false)
  ]
  particleSpawnScript [&InitializeParticle(Color=<rgba 3 0 0.05 1>, Lifetime=$User.Duration, Mesh Scale Mode=::NewEnumerator3, UsePositionOffset=false, Write Material Random=true)]
  particleUpdateScript [
   &ParticleState()
   &ScaleMeshSize(Scale Factor=&MakeVector(X=$System.BeamTile, Y=1, Z=1))
   /Niagara/Modules/Update/Orientation/V2/OrientMeshToVector(Look At Direction:Vector3f=$System.FireDirection, Override Quat=false, UseOptionalSideAxis=false, UseOptionalUpAxis=false)
   &DynamicMaterialParameters(
    Index 0 Param 1=&FloatFromCurve(FloatCurve=DataInterface() {propertyValues {Curve {keys [
         {arriveTangent=-2.8001, interpMode=RCIM_Cubic, leaveTangent=-2.8001, tangentMode=RCTM_User, time=0, value=1.1}
         {arriveTangent=-1, interpMode=RCIM_Cubic, leaveTangent=-1, time=0.1778, value=0.7}
         {arriveTangent=-0.8514, interpMode=RCIM_Cubic, leaveTangent=-0.8514, time=0.6, value=0.5}
         <at 1 0 RCIM_Cubic>
        ]}}})
    Index 0 Param 2=&FloatFromCurve(
     FloatCurve=DataInterface() {propertyValues {Curve {keys [
         {arriveTangent=4.8327, interpMode=RCIM_Cubic, leaveTangent=4.8327, tangentMode=RCTM_User, time=0, value=0}
         {arriveTangent=0.4618, interpMode=RCIM_Cubic, leaveTangent=0.4618, tangentMode=RCTM_User, time=0.35, value=0.8}
         {arriveTangent=0.1596, interpMode=RCIM_Cubic, leaveTangent=0.1596, tangentMode=RCTM_User, time=1, value=1}
        ]}}}
     Scale Curve=-11
    )
    Index 0 Param 3=&FloatFromCurve(FloatCurve=DataInterface() {propertyValues {Curve {keys [<at 0.07 1>,<at 0.25 0>]}}}, CurveIndex=$Particles.Age)
    Index 0 Param 4=$System.BeamTile
    Index 1 Param 1=&FloatFromCurve(FloatCurve=DataInterface() {propertyValues {Curve {keys [
         <at 0.01 1.25 RCIM_Cubic>
         {
          arriveTangent=-0.5563
          arriveTangentWeight=0.0199
          interpMode=RCIM_Cubic
          leaveTangent=-0.5563
          leaveTangentWeight=0.1197
          tangentMode=RCTM_User
          time=0.02
          value=0.7
         }
         {arriveTangent=2.2358, interpMode=RCIM_Cubic, leaveTangent=2.2358, tangentMode=RCTM_User, time=0.25, value=0.77}
         {interpMode=RCIM_Cubic, leaveTangent=1.7664, time=1, value=1.4}
        ]}}})
    Module.Param0WriteEnabled=true
    Module.Param0WriteEnabled1=true
    Module.Param1WriteEnabled=true
    Module.Param1WriteEnabled1=false
    Module.Param2WriteEnabled=true
    Module.Param2WriteEnabled1=false
    Module.Param3WriteEnabled=true
    Module.Param3WriteEnabled1=false
    Write Parameter Index 1=true
   )
   &ScaleColor(
    Scale RGB=&VectorFromCurve(
     VectorCurve=DataInterface() {propertyValues {
       XCurve {keys [
         {arriveTangent=-8.9595, interpMode=RCIM_Cubic, leaveTangent=-8.9595, tangentMode=RCTM_User, time=0.05, value=1}
         {arriveTangentWeight=0.0395, interpMode=RCIM_Cubic, leaveTangentWeight=0.1864, tangentMode=RCTM_User, time=0.2, value=0.25}
         <at 1 0>
        ]}
       YCurve {keys [
         {arriveTangent=-10.2466, interpMode=RCIM_Cubic, leaveTangent=-10.2466, tangentMode=RCTM_User, time=0.05, value=1}
         {arriveTangentWeight=0.0233, leaveTangentWeight=0.11, tangentMode=RCTM_User, time=0.2, value=0.25}
         <at 1 0>
        ]}
       ZCurve {keys [
         {arriveTangent=-10.2466, interpMode=RCIM_Cubic, leaveTangent=-10.2466, tangentMode=RCTM_User, time=0.05, value=1}
         {arriveTangentWeight=0.0233, leaveTangentWeight=0.11, tangentMode=RCTM_User, time=0.2, value=0.25}
         <at 1 0>
        ]}
      }}
     CurveIndex=$Particles.Age
    )
    ScaleA=true
    ScaleRGB=true
   ) {module=ScaleColor001}
  ]
  rendererClasses [&NiagaraMeshRendererProperties]
  renderers [{rendererClass=&NiagaraMeshRendererProperties, rendererIndex=0}]
 }
 BeamResidual {
  simTarget=GPUComputeSim
  emitterUpdateScript [
   &EmitterState()
   &SpawnBurst_Instantaneous(
    Spawn Count=&Multiply_Int(A=&MakeIntFromFloat(Float=$System.BeamTile), B=32)
    Spawn Time=0.1
    Use Spawn Probability=false
   )
  ]
  particleSpawnScript [
   /SetVariables(Particles.SparkIntensity:NiagaraFloat=&FloatFromCurve(
     FloatCurve=DataInterface() {propertyValues {Curve {keys [
         <at 0 1 RCIM_Cubic>
         {arriveTangent=-2.5317, interpMode=RCIM_Cubic, leaveTangent=-2.5317, tangentMode=RCTM_User, time=0.6, value=0.5}
         <at 1 0 RCIM_Cubic>
        ]}}}
     CurveIndex=&RandomRangeFloat(Minimum=0, Maximum=1)
    )) {isSetParameters=true}
   &InitializeParticle(
    Color=&LerpLinearColors(StartColor=<rgba 5 0.0937 0.1754 1>, EndColor=<rgba 15 0.0781 0.3268 1>, LerpFactor=$Particles.SparkIntensity)
    Lifetime=&Multiply_Float(A=$User.Duration, B=&RandomRangeFloat(Minimum=0.8, Maximum=1.1))
    Position=&Add_Vector(
     A=&Multiply_VectorByFloat(Vector=$System.FireDirection, Float=-150)
     B=&Lerp_Vector(
      A=Linked() {linkedVariable {name=Engine.Owner.Position, type=NiagaraPosition}}
      B=$User.BeamEnd_0
      Alpha=&RandomRangeFloat(Minimum=0, Maximum=1)
     )
    )
    Position Mode=::NewEnumerator1
    Sprite Size Mode=::NewEnumerator3
    UsePositionOffset=false
    Write Material Random=true
   )
   &AddVelocity(
    Scale Added Velocity=&VectorFromFloat(Value=&RandomRangeFloat(Minimum=800, Maximum=500))
    Velocity=$System.FireDirection
   )
   &DynamicMaterialParameters(
    Index 0 Param 1=5
    Index 0 Param 3=2
    Module.Param0WriteEnabled=true
    Module.Param1WriteEnabled=false
    Module.Param2WriteEnabled=true
    Module.Param3WriteEnabled=false
   )
  ]
  particleUpdateScript [
   &ParticleState()
   &ScaleColor(
    Linear Color Curve=DataInterface() {propertyValues {
      AlphaCurve {keys [<at 0 1>,<at 1 1>]}
      BlueCurve {keys [
        {arriveTangent=7.4134, interpMode=RCIM_Cubic, time=0, value=0}
        {arriveTangent=-1.452, interpMode=RCIM_Cubic, leaveTangent=-1.452, tangentMode=RCTM_User, time=0.32, value=1}
        <at 1 0.35>
       ]}
      GreenCurve {keys [
        {arriveTangent=7.4134, interpMode=RCIM_Cubic, time=0, value=0}
        {arriveTangent=-1.4408, interpMode=RCIM_Cubic, leaveTangent=-1.4408, tangentMode=RCTM_User, time=0.32, value=1}
        <at 1 0.15>
       ]}
      RedCurve {keys [
        {arriveTangent=7.4134, interpMode=RCIM_Cubic, time=0, value=0}
        {arriveTangent=-1.415, interpMode=RCIM_Cubic, leaveTangent=-1.415, tangentMode=RCTM_User, time=0.32, value=1}
        <at 1 0.15>
       ]}
     }}
    Scale Mode=::NewEnumerator2
   )
   &Drag(
    Drag=&FloatFromCurve(
     FloatCurve=DataInterface() {propertyValues {Curve {keys [{interpMode=RCIM_Cubic, tangentMode=RCTM_User, time=0, value=1}, {interpMode=RCIM_Cubic, tangentMode=RCTM_User, time=0.5, value=0.25}]}}}
     Scale Curve=&FloatFromCurve(FloatCurve=DataInterface() {propertyValues {Curve {keys [<at 0 4>,<at 5 2>]}}}, CurveIndex=$User.Duration)
    )
    Use Linear Drag=true
    Use Rotational Drag=false
   )
   &WindForce(Wind Speed=$System.FireDirection, Wind Speed Scale=50)
   &VectorNoiseForce(
    Force Amount=&FloatFromCurve(
     FloatCurve=DataInterface() {propertyValues {Curve {keys [
         {interpMode=RCIM_Cubic, tangentMode=RCTM_User, time=0, value=0}
         {arriveTangent=2.2381, interpMode=RCIM_Cubic, leaveTangent=2.2381, tangentMode=RCTM_User, time=1, value=1}
        ]}}}
     Scale Curve=200
    )
    Module.Use Random Noise Time Interval=false
    Random Noise Time Interval=0.1
   )
   &SolveForcesAndVelocity(Clamp Velocity=false, Limit Acceleration=false)
   &ScaleSpriteSizeBySpeed(
    Max Scale Factor=<1.5 8>
    Min Scale Factor=<0.5 0.5>
    Sample Scale Factor By Curve=true
    Scale Factor Curve=DataInterface() {propertyValues {Curve {keys [
        {arriveTangent=0.2757, time=0, value=0}
        {arriveTangent=0.6869, interpMode=RCIM_Cubic, leaveTangent=0.6869, tangentMode=RCTM_User, time=0.3, value=0.2}
        {arriveTangent=2.2765, interpMode=RCIM_Cubic, leaveTangent=2.2765, tangentMode=RCTM_User, time=1, value=1}
       ]}}}
    Velocity Threshold=500
   )
   &ScaleSpriteSize(Scale Factor=&Vector2DFromFloat(Value=&FloatFromCurve(FloatCurve=DataInterface() {propertyValues {Curve {keys [
          {arriveTangent=6.1661, interpMode=RCIM_Cubic, time=0, value=0}
          {arriveTangentWeight=0.1067, interpMode=RCIM_Cubic, leaveTangentWeight=0.3202, tangentMode=RCTM_User, time=0.3, value=1}
          {arriveTangent=-3.9956, interpMode=RCIM_Cubic, leaveTangent=-3.9956, tangentMode=RCTM_User, time=1, value=0}
         ]}}})))
   &KillParticles(Kill Particles=&SetBoolByFloatComparison(
     A=&VectorLength(Vector=&Subtract_Vector(
       A=Linked() {linkedVariable {name=Particles.Position, type=NiagaraPosition}}
       B=Linked() {linkedVariable {name=Engine.Owner.Position, type=NiagaraPosition}}
      ))
     B=&Multiply_Float(A=$System.BeamTile, B=500)
    ))
  ]
  rendererClasses [&NiagaraSpriteRendererProperties]
  renderers [{rendererClass=&NiagaraSpriteRendererProperties}]
 }
 NE_FN_Bullet_Tracer {
  emitterSpawnScript [/SetVariables(Emitter.BeamEnd:Vector3f=$User.BeamEnd_0, Emitter.DebugTimeMultiplier=1) {isSetParameters=true}]
  emitterUpdateScript [
   &EmitterState()
   &SpawnBurst_Instantaneous(Spawn Count=1, Use Spawn Probability=false)
  ]
  particleSpawnScript [
   /SetVariables(
    Particles.BeamTile:NiagaraFloat=&Divide_Float(
     A=/Niagara/DynamicInputs/TypeConversions/VectorLengthSafe(Vector:Vector3f=&Subtract_Vector(A=Linked() {linkedVariable {name=Engine.Owner.Position, type=NiagaraPosition}}, B=$Emitter.BeamEnd))
     B=500
    )
    Particles.FireDirection:Vector3f=/Niagara/DynamicInputs/Helpers/SafeNormalizeVector(
     Vector To Normalize:Vector3f=&Subtract_Vector(A=$Emitter.BeamEnd, B=Linked() {linkedVariable {name=Engine.Owner.Position, type=NiagaraPosition}})
     Fallback Vector:Vector3f=$Engine.Owner.SystemXAxis
    )
   ) {isSetParameters=true}
   /SetVariables(
    Particles.BeamDistanceScalar:NiagaraFloat=&Lerp_Float(
     A=0.35
     B=1
     Alpha=&FloatFromCurve(
      FloatCurve=DataInterface() {propertyValues {Curve {keys [
          {arriveTangent=0.0002, interpMode=RCIM_Cubic, leaveTangent=0.0002, tangentMode=RCTM_User, time=0, value=0}
          {arriveTangent=0.0001, interpMode=RCIM_Cubic, leaveTangent=0.0001, tangentMode=RCTM_User, time=2500, value=0.35}
          <at 15000 0.8>
          {interpMode=RCIM_Cubic, tangentMode=RCTM_User, time=35000, value=1}
         ]}}}
      CurveIndex=&VectorLengthSafe(Vector:Vector3f=&Subtract_Vector(A=Linked() {linkedVariable {name=Engine.Owner.Position, type=NiagaraPosition}}, B=$Emitter.BeamEnd))
     )
    )
    Particles.LODDistanceRemap:NiagaraFloat=&FloatFromCurve(FloatCurve=DataInterface() {propertyValues {Curve {keys [<at 500 0>,<at 2500 1>]}}}, CurveIndex=$Engine.Owner.LODDistance)
   ) {isSetParameters=true}
   &InitializeParticle(
    Color=<rgba 25 0.5361 7.0598 1>
    Lifetime=&Multiply_Float(
     A=&Multiply_Float(A=&ClampFloat(Float=$Particles.BeamDistanceScalar, Min=0.5), B=1)
     B=$Emitter.DebugTimeMultiplier
    )
    Mesh Scale Mode=::NewEnumerator3
    Sprite Size=<100 700>
    Sprite Size Mode=::NewEnumerator4
    UsePositionOffset=false
    Write Material Random=true
   )
   &DynamicMaterialParameters(
    Index 0 Param 1=8
    Index 0 Param 2=50
    Index 0 Param 3=5
    Module.Param0WriteEnabled=true
    Module.Param1WriteEnabled=true
    Module.Param2WriteEnabled=true
    Module.Param3WriteEnabled=false
   ) {module=DynamicMaterialParameters001}
   &OrientMeshToVector(Look At Direction:Vector3f=$Particles.FireDirection, Override Quat=false, UseOptionalSideAxis=false, UseOptionalUpAxis=false) {enabled=false}
   /SetVariables(Particles.BulletDynParams:Vector4f=<1 1 1 1>) {enabled=false, isSetParameters=true}
   /SetVariables(Particles.SpriteAlignment=$Particles.FireDirection) {isSetParameters=true}
  ]
  particleUpdateScript [
   &ParticleState()
   /Niagara/Modules/Update/Utility/FrameCounter()
   /SetVariables(Particles.FrameLimitedAge:NiagaraFloat=&self:NE_FN_Bullet_Tracer.NiagaraScratchPadContainer_1.ScratchDynamicInput(
     MaxDuration:NiagaraFloat=0.35
     MinDuration:NiagaraFloat=0.1
     CurrentFrame:NiagaraFloat=&MakeFloatFromInt(Int=$Particles.FrameCounter.Counter)
     MaxFrames:NiagaraFloat=&MakeFloatFromInt(Int=10)
     DurationMultiplier:NiagaraFloat=$Particles.BeamDistanceScalar
     DebugTimeMultiplier:NiagaraFloat=$Emitter.DebugTimeMultiplier
    )) {isSetParameters=true}
   /SetVariables(Particles.Travel:NiagaraFloat=&self:NE_FN_Bullet_Tracer.NiagaraScratchPadContainer_1.RemapRange(
     Value:NiagaraFloat=$Particles.FrameLimitedAge
     Input Min:NiagaraFloat=0
     Input Max:NiagaraFloat=1
     Output Min:NiagaraFloat=0.01
     Output Max:NiagaraFloat=1
    )) {isSetParameters=true}
   /SetVariables(Particles.BulletPosition:Vector3f=&Lerp_Vector(A=Linked() {linkedVariable {name=Particles.Initial.Position, type=NiagaraPosition}}, B=$Emitter.BeamEnd, Alpha=$Particles.Travel)) {isSetParameters=true}
   /SetVariables(Particles.BulletSize:Vector3f=&Multiply_VectorByFloat(
     Vector=&Lerp_Vector(A=<2.5 1.25 1.25>, B=<4 2 2>, Alpha=$Particles.LODDistanceRemap)
     Float=&FloatFromCurve(
      FloatCurve=DataInterface() {propertyValues {Curve {keys [{arriveTangent=32.8701, time=0, value=0}, <at 0.05 1>]}}}
      CurveIndex=$Particles.FrameLimitedAge
     )
    )) {isSetParameters=true}
   &ScaleSpriteSize(Scale Factor=&Vector2DFromFloat(Value=&FloatFromCurve(
      FloatCurve=DataInterface() {propertyValues {Curve {keys [
          {arriveTangent=23.5738, interpMode=RCIM_Cubic, leaveTangent=23.5738, tangentMode=RCTM_User, time=0, value=0}
          {arriveTangentWeight=0.0833, interpMode=RCIM_Cubic, leaveTangentWeight=0.1667, tangentMode=RCTM_User, time=0.1, value=1}
          {arriveTangentWeight=0.2778, interpMode=RCIM_Cubic, leaveTangentWeight=0.1389, tangentMode=RCTM_User, time=0.75, value=1}
          {arriveTangent=-9.5905, interpMode=RCIM_Cubic, leaveTangent=-9.5905, tangentMode=RCTM_User, time=1, value=0}
         ]}}}
      CurveIndex=$Particles.FrameLimitedAge
     )))
   &ScaleMeshSize(Scale Factor=&MakeVector(X=&Multiply_Float(A=$Particles.Travel, B=$Particles.BeamTile), Y=1, Z=1)) {enabled=false}
   &DynamicMaterialParameters(
    Index 0 Param 1=&FloatFromCurve(FloatCurve=DataInterface() {propertyValues {Curve {keys [
         {arriveTangent=-0.4135, interpMode=RCIM_Cubic, leaveTangent=-0.4135, tangentMode=RCTM_User, time=0, value=1}
         {arriveTangent=-1.3177, interpMode=RCIM_Cubic, leaveTangent=-1.3177, tangentMode=RCTM_User, time=1, value=0}
        ]}}})
    Index 0 Param 2=&FloatFromCurve(
     FloatCurve=DataInterface() {propertyValues {Curve {keys [
         {arriveTangent=5.2627, interpMode=RCIM_Cubic, leaveTangent=5.2627, tangentMode=RCTM_User, time=0, value=0}
         {arriveTangent=0.8931, interpMode=RCIM_Cubic, leaveTangent=0.8931, tangentMode=RCTM_User, time=0.2287, value=0.7}
         {arriveTangentWeight=0.2571, interpMode=RCIM_Cubic, tangentMode=RCTM_User, time=1, value=1}
        ]}}}
     Scale Curve=&Multiply_Float(
      A=&Lerp_Float(A=0.5, B=1, Alpha=$Particles.BeamDistanceScalar)
      B=&RandomRangeFloat(Minimum=-10, Maximum=-8)
     )
    )
    Index 0 Param 3=&FloatFromCurve(
     FloatCurve=DataInterface() {propertyValues {Curve {keys [
         {arriveTangent=-0.9882, time=0.2, value=1}
         {arriveTangent=-1.2626, interpMode=RCIM_Cubic, leaveTangent=-1.2626, tangentMode=RCTM_User, time=0.5, value=0.6}
         {arriveTangent=-10.4271, interpMode=RCIM_Cubic, leaveTangent=-10.4271, tangentMode=RCTM_User, time=0.6, value=0}
        ]}}}
     CurveIndex=$Particles.FrameLimitedAge
    )
    Index 0 Param 4=&Multiply_Float(
     A=&MakeFloatFromVector(Vector=$Particles.Scale, Channel=::NewEnumerator0)
     B=&Lerp_Float(A=0.8, B=1.35, Alpha=$Particles.LODDistanceRemap)
    )
    Index 1 Param 1=&FloatFromCurve(
     FloatCurve=DataInterface() {propertyValues {Curve {keys [
         {arriveTangent=-0.8182, interpMode=RCIM_Cubic, leaveTangent=-0.8182, tangentMode=RCTM_User, time=0, value=0.7}
         {arriveTangent=2.4059, interpMode=RCIM_Cubic, leaveTangent=9.91, tangentMode=RCTM_Break, time=0.5, value=0.87}
         {arriveTangent=0.8994, interpMode=RCIM_Cubic, leaveTangent=0.8994, tangentMode=RCTM_User, time=0.5714, value=1.214}
         {arriveTangent=0.8128, interpMode=RCIM_Cubic, leaveTangent=0.8128, tangentMode=RCTM_User, time=1, value=1.5}
        ]}}}
     Scale Curve=&Lerp_Float(A=0.35, B=0.6, Alpha=$Particles.LODDistanceRemap)
    )
    Module.Param0WriteEnabled=true
    Module.Param0WriteEnabled1=true
    Module.Param1WriteEnabled=true
    Module.Param1WriteEnabled1=false
    Module.Param2WriteEnabled=true
    Module.Param2WriteEnabled1=false
    Module.Param3WriteEnabled=false
    Module.Param3WriteEnabled1=false
    Write Parameter Index 1=true
   ) {enabled=false}
   &ScaleColor(
    Scale Alpha=&FloatFromCurve(
     FloatCurve=DataInterface() {propertyValues {Curve {keys [
         {arriveTangent=5.6806, interpMode=RCIM_Cubic, leaveTangent=5.6806, tangentMode=RCTM_User, time=0, value=0}
         {interpMode=RCIM_Cubic, tangentMode=RCTM_User, time=0.5, value=1}
         {arriveTangent=-4.9856, interpMode=RCIM_Cubic, leaveTangent=-4.9856, tangentMode=RCTM_User, time=1, value=0}
        ]}}}
     CurveIndex=$Particles.FrameLimitedAge
    )
    ScaleA=true
    ScaleRGB=false
   )
   &KillParticles(Kill Particles=&SetBoolByFloatComparison(A=$Particles.FrameLimitedAge, B=1, Comparison Type=::NewEnumerator2))
  ]
  rendererClasses [
   &NiagaraMeshRendererProperties
   &NiagaraSpriteRendererProperties
   &NiagaraLightRendererProperties
  ]
  renderers [
   {rendererClass=&NiagaraMeshRendererProperties, rendererIndex=0}
   {rendererClass=&NiagaraMeshRendererProperties, rendererIndex=1}
   {rendererClass=&NiagaraSpriteRendererProperties, rendererIndex=2}
   {rendererClass=&NiagaraLightRendererProperties, rendererIndex=3}
  ]
 }
 NE_FN_Sparks {
  simTarget=GPUComputeSim
  emitterUpdateScript [
   &EmitterState(Life Cycle Mode=::NewEnumerator1, Loop Behavior=::NewEnumerator1, Loop Duration=0.2, UseLoopDelay=false)
   &SpawnBurst_Instantaneous(
    Spawn Count=&Multiply_Int(A=&MakeIntFromFloat(Float=$System.BeamTile), B=30)
    Use Spawn Probability=false
   )
   &SpawnRate(SpawnRate=&Multiply_Float(A=$System.BeamTile, B=200), Use Spawn Probability=false)
  ]
  particleSpawnScript [
   /SetVariables(Particles.SparkIntensity:NiagaraFloat=&FloatFromCurve(
     FloatCurve=DataInterface() {propertyValues {Curve {keys [
         <at 0 1 RCIM_Cubic>
         {arriveTangent=-2.5317, interpMode=RCIM_Cubic, leaveTangent=-2.5317, tangentMode=RCTM_User, time=0.6, value=0.5}
         <at 1 0 RCIM_Cubic>
        ]}}}
     CurveIndex=&RandomRangeFloat(Minimum=0, Maximum=1)
    )) {isSetParameters=true}
   &InitializeParticle(
    Color=&LerpLinearColors(StartColor=<rgba 10 0.7087 0.8636 1>, EndColor=<rgba 25 0.543 0.9506 1>, LerpFactor=$Particles.SparkIntensity)
    Lifetime Max=1.1
    Lifetime Min=0.4
    Lifetime Mode=::NewEnumerator1
    Mass=&Lerp_Float(A=0.3, B=1, Alpha=$Particles.SparkIntensity)
    Mass Mode=::NewEnumerator1
    Position=&Lerp_Vector(
     A=Linked() {linkedVariable {name=Engine.Owner.Position, type=NiagaraPosition}}
     B=&Add_Vector(A=$User.BeamEnd_0, B=&Multiply_VectorByFloat(Vector=$System.FireDirection, Float=-300))
     Alpha=&RandomRangeFloat(Minimum=0, Maximum=1)
    )
    Position Mode=::NewEnumerator1
    Sprite Size Mode=::NewEnumerator3
    Uniform Sprite Size=&Multiply_Float(
     A=&FN_LOD_Distance_Remap(Start Distance=100, End Distance=2000, Minimum Output=1, Maximum Output=2)
     B=&Lerp_Float(A=2, B=3, Alpha=$Particles.SparkIntensity)
    )
    UsePositionOffset=false
    Write Material Random=true
   )
   /SetVariables(Particles.SpriteSize=&Multiply_Vector2(
     A=$Particles.SpriteSize
     B=&Vector2DFromFloat(Value=&FloatFromCurve(FloatCurve=DataInterface() {propertyValues {Curve {keys [<at 1000 1 RCIM_Cubic>,<at 5000 2>]}}}, CurveIndex=$Engine.Owner.LODDistance))
    )) {enabled=false, isSetParameters=true}
   &SphereLocation(
    MaskSpawnGroup=false
    Override Seed=false
    Sphere Radius=3
    Surface Only=false
    UseNonUniformScale=false
    UseOffset=false
    UseSphereOrientation=false
   )
   &AddVelocityInCone(
    Cone Angle=8
    Cone Axis=$System.FireDirection
    Module.Use Velocity Falloff On Cone Axis=false
    Velocity Strength=&RandomRangeFloat(Minimum=500, Maximum=2000)
   )
  ]
  particleUpdateScript [
   &ParticleState()
   &GravityForce(Gravity=<0 0 -1600>) {enabled=false}
   &Drag(Drag=&RandomRangeFloat(Minimum=3, Maximum=6), Use Linear Drag=true, Use Rotational Drag=false)
   &VectorNoiseForce(
    Force Amount=&FloatFromCurve(
     FloatCurve=DataInterface() {propertyValues {Curve {keys [
         {interpMode=RCIM_Cubic, tangentMode=RCTM_User, time=0, value=0}
         <at 0.5 1 RCIM_Cubic>
         {arriveTangent=-4.4143, interpMode=RCIM_Cubic, leaveTangent=-4.4143, tangentMode=RCTM_User, time=1, value=0}
        ]}}}
     CurveIndex=&Multiply_Float(A=$System.Age, B=0.5)
     Scale Curve=2000
    )
    Module.Use Random Noise Time Interval=true
   )
   /Niagara/Modules/Update/Forces/LineAttractionForce(
    Attraction Strength:NiagaraFloat=&FloatFromCurve(FloatCurve=DataInterface() {propertyValues {Curve {keys [<at 0 1 RCIM_Cubic>,<at 0.6 0 RCIM_Cubic>]}}}, Scale Curve=1000)
    Line End:Vector3f=&Add_Vector(A=$User.BeamEnd_0, B=&Multiply_VectorByFloat(Vector=$System.FireDirection, Float=500))
    Line Start:NiagaraPosition=$Engine.Owner.Position
    Use Attraction Falloff=false
   )
   &SolveForcesAndVelocity(Clamp Velocity=false, Limit Acceleration=false)
   &ScaleColor(
    Linear Color Curve=DataInterface() {propertyValues {
      AlphaCurve {keys [<at 0 1>,<at 1 1>]}
      BlueCurve {keys [
        <at 0 2 RCIM_Cubic>
        {arriveTangent=-1.452, interpMode=RCIM_Cubic, leaveTangent=-1.452, tangentMode=RCTM_User, time=0.257, value=1}
        <at 1 0.35>
       ]}
      GreenCurve {keys [
        <at 0 2 RCIM_Cubic>
        {arriveTangent=-1.4408, interpMode=RCIM_Cubic, leaveTangent=-1.4408, tangentMode=RCTM_User, time=0.257, value=1}
        <at 1 0.15>
       ]}
      RedCurve {keys [
        <at 0 2 RCIM_Cubic>
        {arriveTangent=-1.415, interpMode=RCIM_Cubic, leaveTangent=-1.415, tangentMode=RCTM_User, time=0.257, value=1}
        <at 1 0.15>
       ]}
     }}
    Scale Mode=::NewEnumerator2
   )
   &ScaleSpriteSize(Scale Factor=&Vector2DFromFloat(Value=&FloatFromCurve(FloatCurve=DataInterface() {propertyValues {Curve {keys [<at 0.75 1 RCIM_Cubic>, {arriveTangent=-12.0083, interpMode=RCIM_Cubic, leaveTangent=-12.0083, tangentMode=RCTM_User, time=1, value=0}]}}})))
   &ScaleSpriteSizeBySpeed(
    Max Scale Factor=<1 10>
    Sample Scale Factor By Curve=true
    Scale Factor Curve=DataInterface() {propertyValues {Curve {keys [
        {arriveTangent=0.2757, time=0, value=0}
        {arriveTangent=0.6869, interpMode=RCIM_Cubic, leaveTangent=0.6869, tangentMode=RCTM_User, time=0.3, value=0.2}
        {arriveTangent=2.2765, interpMode=RCIM_Cubic, leaveTangent=2.2765, tangentMode=RCTM_User, time=1, value=1}
       ]}}}
    Velocity Threshold=1400
   )
   &KillParticles(Kill Particles=&SetBoolByFloatComparison(
     A=&VectorLength(Vector=&Subtract_Vector(
       A=Linked() {linkedVariable {name=Particles.Position, type=NiagaraPosition}}
       B=Linked() {linkedVariable {name=Engine.Owner.Position, type=NiagaraPosition}}
      ))
     B=&Multiply_Float(A=$System.BeamTile, B=500)
    ))
  ]
  rendererClasses [&NiagaraSpriteRendererProperties]
  renderers [{rendererClass=&NiagaraSpriteRendererProperties}]
 }
}
summary {
 systemName=NS_ReactorGrade_BeamFire
 emitters [
  {emitterName=Beam, rendererClasses [&NiagaraMeshRendererProperties]}
  {emitterName=NE_FN_Sparks, simTarget=GPUComputeSim, rendererClasses [&NiagaraSpriteRendererProperties]}
  {emitterName=BeamResidual, simTarget=GPUComputeSim, rendererClasses [&NiagaraSpriteRendererProperties]}
  {
   emitterName=NE_FN_Bullet_Tracer
   rendererClasses [
    &NiagaraMeshRendererProperties
    &NiagaraSpriteRendererProperties
    &NiagaraLightRendererProperties
   ]
  }
 ]
 userVariables [
  User.BeamEnd_0 {type=Vector3f, description="", defaultValue {struct=Vector3f, value=<3000 0 2000>}}
  User.Duration {type=NiagaraFloat, defaultValue=4, description=""}
  User.MinWarmupTime {type=NiagaraFloat, defaultValue=0, description=""}
 ]
}
